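# Niagara system for the ManAndMonster emote: a spinning beaker mesh that falls under gravity and dies on collision, with glass, goop, dust, smoke and ribbon-trail effects.
# Two beaker emitters scale burst counts by front-end state and emit location and collision events.
# Beaker uses a collision trace channel, while Beaker_Frontend sets an analytical plane from a skeletal mesh sample.
# The secondary emitters have no active spawn modules, suggesting they are event-driven.
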NiagaraSystem /BRCosmetics/Athena/Effects/Particles/Emotes/NS_Emote_ManAndMonster/NS_Emote_ManAndMonster_Beaker_01
dependencies {
 usedDataInterfaces [&NiagaraDataInterfaceCurve, &NiagaraDataInterfaceColorCurve]
 usedDynamicInputs []
 usedModules [&SystemState]
}
emitters {
 Beaker {
  emitterUpdateScript [
   &EmitterState(Life Cycle Mode=::NewEnumerator1, Loop Behavior=::NewEnumerator1, Loop Duration=1, UseLoopDelay=false)
   &SpawnRate(SpawnRate=5, Use Spawn Probability=false) {enabled=false}
   &SpawnBurst_Instantaneous(
    Spawn Count=&MakeIntFromFloat(Float=&FN_FrontEndFloatModifier(Gameplay=1, Front End=0))
    Use Spawn Probability=false
   )
  ]
  particleSpawnScript [
   /Niagara/Modules/Spawn/Initialization/InitializeParticle(
    Module.Write Color=true
    Module.Write Lifetime=true
    Module.Write Mass=true
    Module.Write Material Random=true
    Module.Write Position=false
    Module.Write Scale=true
    Module.Write SpriteRotation=true
    Module.Write SpriteSize=true
    Position=&SimulationPosition()
   )
   /Niagara/Modules/Spawn/Orientation/V2/InitialMeshOrientation(Rotation=<0 0.02 0>, Use Orientation Vector=true, Use Rotation Vector=true)
  ]
  particleUpdateScript [
   &ParticleState()
   &AddVelocity(Coordinate Space=::World, Velocity=<5 0 -40>)
   &Drag(Drag=2, Use Linear Drag=true, Use Rotational Drag=false)
   &MeshRotationForce(Rotation=<0 20 0>)
   &Collision(
    Advanced Aging Rate=1
    Age Colliding Particles=false
    CPU Collision Trace Channel=ECC_GameTraceChannel11
    Module.Control Rest State Via Surface Normal=false
    Module.EnableMaxCollisionCount=false
    Module.Randomize Collision Normal Vector=false
    Radius Calculation Type=::NewEnumerator1
   )
   &GravityForce(Gravity=<500 0 -1280>)
   &SolveForcesAndVelocity(Clamp Velocity=false, Limit Acceleration=false)
   &SolveRotationalForcesAndVelocity()
   &GenerateLocationEvent(Module.Use Event Send Rate=true)
   /Niagara/Modules/Events/GenerateCollisionEvent(Module.Use Time Between Events=false, Module.Use Velocity Threshold=false, Velocity Threshold=0)
   &KillParticles(Kill Particles=$Particles.HasCollided)
  ]
  rendererClasses [&NiagaraMeshRendererProperties]
  renderers [{rendererClass=&NiagaraMeshRendererProperties, rendererIndex=0}]
 }
 Beaker_Frontend {
  emitterUpdateScript [
   &EmitterState(Life Cycle Mode=::NewEnumerator1, Loop Behavior=::NewEnumerator1, Loop Duration=1, UseLoopDelay=false)
   &SpawnRate(SpawnRate=5, Use Spawn Probability=false) {enabled=false}
   &SpawnBurst_Instantaneous(
    Spawn Count=&MakeIntFromFloat(Float=&FN_FrontEndFloatModifier(Gameplay=1, Front End=2))
    Use Spawn Probability=false
   )
  ]
  particleSpawnScript [
   &InitializeParticle(
    Module.Write Color=true
    Module.Write Lifetime=true
    Module.Write Mass=true
    Module.Write Material Random=true
    Module.Write Position=false
    Module.Write Scale=true
    Module.Write SpriteRotation=true
    Module.Write SpriteSize=true
    Position=&SimulationPosition()
   )
   &InitialMeshOrientation(Rotation=<0 0.02 0>, Use Orientation Vector=true, Use Rotation Vector=true)
   &SkeletalMeshLocation(
    Apply To Particle Position=false
    Bone Sampling Mode=::NewEnumerator4
    Inherit Velocity=false
    "Inherit Velocity (Scale)"=0.5
    MaskSpawnGroup=true
    Skeletal Mesh=$User.SkeletalMesh
   )
   /SetVariables(Particles.CollisionVector:Vector3f=$Output.SkeletalMeshLocation.SampledPosition) {isSetParameters=true}
  ]
  particleUpdateScript [
   &ParticleState()
   &AddVelocity(Coordinate Space=::World, Velocity=<5 0 -40>)
   &Drag(Drag=2, Use Linear Drag=true, Use Rotational Drag=false)
   &MeshRotationForce(Rotation=<0 20 0>)
   &Collision(
    Advanced Aging Rate=1
    Age Colliding Particles=false
    Analytical Collision Plane Position 1:Vector3f=$Particles.CollisionVector
    CPU Collision Type=::NewEnumerator3
    Module.Control Rest State Via Surface Normal=false
    Module.EnableMaxCollisionCount=false
    Module.Randomize Collision Normal Vector=false
    Perform 2 Analytical Collision Queries=false
    Radius Calculation Type=::NewEnumerator1
   )
   &GravityForce(Gravity=<500 0 -1280>)
   &SolveForcesAndVelocity(Clamp Velocity=false, Limit Acceleration=false)
   &SolveRotationalForcesAndVelocity()
   &GenerateLocationEvent(Module.Use Event Send Rate=true)
   &GenerateCollisionEvent(Module.Use Time Between Events=false, Module.Use Velocity Threshold=false, Velocity Threshold=0)
   &KillParticles(Kill Particles=$Particles.HasCollided)
  ]
  rendererClasses [&NiagaraMeshRendererProperties]
  renderers [{rendererClass=&NiagaraMeshRendererProperties, rendererIndex=0}]
 }
 Dust {
  emitterUpdateScript [
   &EmitterState(Life Cycle Mode=::NewEnumerator1, Loop Behavior=::NewEnumerator1, Loop Duration=1, UseLoopDelay=false)
   &SpawnBurst_Instantaneous(Spawn Count=20, Use Spawn Probability=false) {enabled=false}
  ]
  particleSpawnScript [
   &InitializeParticle(
    Color=&ColorFromCurve(ColorCurve=DataInterface() {propertyValues {
       AlphaCurve {keys [<at -0.0445 0.3>]}
       BlueCurve {keys [<at 0 1.5 RCIM_Cubic>,<at 0.0727 0.3073>,<at 0.3488 0.1094>]}
       GreenCurve {keys [<at 0 0.85 RCIM_Cubic>,<at 0.0727 0.1076>,<at 0.3488 0.0055>]}
       RedCurve {keys [<at 0 0.2 RCIM_Cubic>,<at 0.0727 0.041>,<at 0.3488 0.0055>]}
      }})
    Lifetime=&UniformRangedFloat(Minimum=0.4, Maximum=0.8)
    Module.Write Color=true
    Module.Write Lifetime=true
    Module.Write Mass=true
    Module.Write Material Random=true
    Module.Write Position=false
    Module.Write Scale=false
    Module.Write SpriteRotation=true
    Module.Write SpriteSize=true
    Position=&SimulationPosition()
    Sprite Rotation=&UniformRangedFloat(Minimum=-8, Maximum=8)
    Sprite Size=&Vector2DFromFloat(Value=&UniformRangedFloat(Minimum=25, Maximum=30))
   )
   &CylinderLocation(
    Cylinder Height=0
    Cylinder Radius=3
    Lathe Profile=DataInterface() {propertyValues {Curve {keys [
        {arriveTangent=2.9957, interpMode=RCIM_Cubic, leaveTangent=2.9957, tangentMode=RCTM_User, time=0, value=0.451}
        {arriveTangent=0.049, interpMode=RCIM_Cubic, leaveTangent=0.049, time=0.5, value=1}
        {arriveTangent=-2.4251, interpMode=RCIM_Cubic, leaveTangent=-2.4251, tangentMode=RCTM_User, time=1, value=0.5}
       ]}}}
    MaskSpawnGroup=false
    Module.Enable Lathe Profile=true
    Module.Override Randomness=false
    Module.Surface Only=false
    Offset=<0 0 1>
   )
   &AddVelocityFromPoint(
    Module.Use Falloff Distance=false
    Origin Offset=<0 0 1>
    Velocity Origin:Vector3f=&SimulationPosition()
    Velocity Strength=&UniformRangedFloat(Minimum=200, Maximum=180)
   )
   &ApplyInitialForces()
   &CameraOffset(Camera Offset Amount=10)
  ]
  particleUpdateScript [
   &ParticleState()
   &AccelerationForce(Acceleration=<0 0 50>)
   &Drag(Drag=7, Use Linear Drag=true, Use Rotational Drag=false)
   &SolveForcesAndVelocity(Clamp Velocity=false, Limit Acceleration=false)
   &ScaleSpriteSize(Scale Factor=&Vector2DFromFloat(Value=&FloatFromCurve(FloatCurve=DataInterface() {propertyValues {Curve {keys [
          {arriveTangent=4.1286, interpMode=RCIM_Cubic, leaveTangent=4.1286, tangentMode=RCTM_User, time=0, value=0}
          {arriveTangent=1.2171, interpMode=RCIM_Cubic, leaveTangent=1.2171, tangentMode=RCTM_User, time=0.309, value=0.8022}
          <at 1 1.25 RCIM_Cubic>
         ]}}})))
   /SetVariables(Particles.SubImageIndex=&FloatFromCurve(FloatCurve=DataInterface() {propertyValues {Curve {keys [
         {arriveTangent=29.8623, interpMode=RCIM_Cubic, leaveTangent=29.8623, tangentMode=RCTM_User, time=-0.0046, value=5.0556}
         {arriveTangent=8.1241, interpMode=RCIM_Cubic, leaveTangent=8.1241, tangentMode=RCTM_User, time=1.0011, value=22.1689}
        ]}}})) {isSetParameters=true, module=SetVariables_#1}
   &DynamicMaterialParameters(
    Index 0 Param 1=&FloatFromCurve(FloatCurve=DataInterface() {propertyValues {Curve {keys [<at -0.0046 1.0113>, {arriveTangent=-1.9998, interpMode=RCIM_Cubic, leaveTangent=-1.9998, time=0.5, value=1}, <at 1.0011 -1>]}}})
    Index 0 Param 4=0.5
    Module.Param0WriteEnabled=true
    Module.Param1WriteEnabled=false
    Module.Param2WriteEnabled=false
    Module.Param3WriteEnabled=true
   ) {module=DynamicMatParams}
   &ScaleColor(
    Scale Alpha=&FloatFromCurve(
     FloatCurve=DataInterface() {propertyValues {Curve {keys [
         {arriveTangent=-3.1757, interpMode=RCIM_Cubic, leaveTangent=-3.1757, tangentMode=RCTM_User, time=0.5, value=1.5}
         {arriveTangent=-1.5121, interpMode=RCIM_Cubic, leaveTangent=-1.5121, tangentMode=RCTM_User, time=1.0011, value=0.002}
        ]}}}
     Scale Curve=0.85
    )
    Scale RGB=&Multiply_VectorByFloat(Vector=<0.3 0.3 0.3>, Float=&RandomRangeFloat(Minimum=0.65, Maximum=0.2))
    ScaleA=true
    ScaleRGB=true
   ) {module=ScaleColor001}
  ]
  rendererClasses [&NiagaraSpriteRendererProperties]
  renderers [{rendererClass=&NiagaraSpriteRendererProperties}]
 }
 GlassCrack {
  emitterUpdateScript [&EmitterState(Life Cycle Mode=::NewEnumerator1, Loop Behavior=::NewEnumerator1, Loop Duration=3, UseLoopDelay=false)]
  particleSpawnScript [
   &InitializeParticle(
    Lifetime=&UniformRangedFloat(Minimum=0.9, Maximum=1.6)
    Mesh Scale=&UniformRangedVector(Minimum=<0.05 0.05 0.05>, Maximum=<0.22 0.22 0.22>)
    Module.Write Color=false
    Module.Write Lifetime=true
    Module.Write Mass=true
    Module.Write Material Random=true
    Module.Write Position=true
    Module.Write Scale=true
    Module.Write SpriteRotation=false
    Module.Write SpriteSize=true
   )
   &AddVelocity(Velocity=&Multiply_Vector(
     A=$Output.SphereLocation.SphereNormal
     B=&VectorFromFloat(Value=&UniformRangedFloat(Minimum=1, Maximum=3))
    ))
   &SphereLocation(
    MaskSpawnGroup=false
    Override Seed=false
    Sphere Radius=10
    Surface Only=false
    UseNonUniformScale=false
    UseOffset=false
    UseSphereOrientation=false
   )
   &InitialMeshOrientation(
    Rotation=&UniformRangedVector(Minimum=<0 0 0>, Maximum=<1 1 1>)
    Use Orientation Vector=true
    Use Rotation Vector=true
   )
   &AddVelocityInCone(
    Cone Angle=100
    Cone Axis=<0 0 1>
    Module.Use Velocity Falloff On Cone Axis=false
    Velocity Strength=&RandomRangeFloat(Minimum=250, Maximum=150)
   )
   &AddVelocity(Velocity=&RandomRangeVector(Minimum=<0 0 50>, Maximum=<1 1 0>)) {module=AddVelocity001}
  ]
  particleUpdateScript [
   &ParticleState()
   &ScaleMeshSize(Scale Factor=&VectorFromFloat(Value=&FloatFromCurve(FloatCurve=DataInterface() {propertyValues {Curve {keys [<at 0.5 2 RCIM_Cubic>,<at 1 0 RCIM_Cubic>]}}})))
   &AccelerationForce(Acceleration=<0 0 -500>)
   /Niagara/Modules/Update/Orientation/V2/UpdateMeshOrientation(
    Rotation Rate:NiagaraFloat=&FloatFromCurve(
     FloatCurve=DataInterface() {propertyValues {Curve {keys [{arriveTangent=-2.8458, interpMode=RCIM_Cubic, leaveTangent=-2.8458, tangentMode=RCTM_User, time=0, value=1}, <at 1 0.2 RCIM_Cubic>]}}}
     Scale Curve=2
    )
    Rotation Vector:Vector3f=<2 0.4 0.1>
   )
   &Drag(Drag=2, Use Linear Drag=true, Use Rotational Drag=false)
   &SolveForcesAndVelocity(Clamp Velocity=false, Limit Acceleration=false)
  ]
  rendererClasses [&NiagaraMeshRendererProperties]
  renderers [{rendererClass=&NiagaraMeshRendererProperties, rendererIndex=0}]
 }
 Goop {
  emitterUpdateScript [
   &EmitterState(Life Cycle Mode=::NewEnumerator1, Loop Duration=1, UseLoopDelay=false)
   &SpawnBurst_Instantaneous(Spawn Count=10, Use Spawn Probability=false) {enabled=false}
  ]
  particleSpawnScript [
   &InitializeParticle(
    Color=&RandomRangeLinearColor(Minimum=<rgba 0.0444 0.2519 0.3333 1>, Maximum=<rgba 0.0118 0.0502 0.0885 1>)
    Lifetime=&RandomRangeFloat(Minimum=0.6, Maximum=0.8)
    Module.Write Color=true
    Module.Write Lifetime=true
    Module.Write Mass=true
    Module.Write Material Random=true
    Module.Write Position=false
    Module.Write Scale=false
    Module.Write SpriteRotation=false
    Module.Write SpriteSize=true
    Position=&SimulationPosition()
    Sprite Size=<5 5>
   )
   &AddVelocityInCone(
    Cone Angle=50
    Cone Axis=<0 0 1>
    Module.Use Velocity Falloff On Cone Axis=false
    Velocity Strength=&RandomRangeFloat(Minimum=350, Maximum=150)
   )
   &InitialMeshOrientation(
    Rotation=&RandomRangeVector(Minimum=<0 0 0>, Maximum=<1 1 1>)
    Use Orientation Vector=true
    Use Rotation Vector=true
   )
  ]
  particleUpdateScript [
   &ParticleState()
   &GravityForce()
   &SolveForcesAndVelocity(Clamp Velocity=false, Limit Acceleration=false)
   &ScaleMeshSize(Scale Factor=&VectorFromFloat(Value=&RandomRangeFloat(Minimum=0.1, Maximum=0.5)))
   &ScaleMeshSize(Scale Factor=&VectorFromFloat(Value=&FloatFromCurve(FloatCurve=DataInterface() {propertyValues {Curve {keys [
          {interpMode=RCIM_Cubic, tangentMode=RCTM_User, time=0, value=1}
          {arriveTangent=-1.9184, interpMode=RCIM_Cubic, leaveTangent=-1.9184, tangentMode=RCTM_User, time=1, value=0}
         ]}}}))) {module=ScaleMeshSize001}
  ]
  rendererClasses [&NiagaraMeshRendererProperties]
  renderers [{rendererClass=&NiagaraMeshRendererProperties, rendererIndex=0}]
 }
 Smoke {
  emitterUpdateScript [
   &EmitterState(Life Cycle Mode=::NewEnumerator1, Loop Behavior=::NewEnumerator1, Loop Duration=0.5, UseLoopDelay=false)
   &SpawnBurst_Instantaneous(Spawn Count=35, Spawn Time=0.05, Use Spawn Probability=false) {enabled=false, module=SpawnBurst_Instantaneous001}
   &SpawnRate(SpawnRate=100, Use Spawn Probability=false) {enabled=false}
  ]
  particleSpawnScript [
   &InitializeParticle(
    Color=&ColorFromCurve(ColorCurve=DataInterface() {propertyValues {
       AlphaCurve {keys [<at -0.0445 0.1>]}
       BlueCurve {keys [<at 0 3 RCIM_Cubic>,<at 0.0727 0.7>,<at 0.6644 0.3>]}
       GreenCurve {keys [<at 0 2.5667 RCIM_Cubic>,<at 0.0727 0.245>,<at 0.6644 0.015>]}
       RedCurve {keys [<at 0 0.4 RCIM_Cubic>,<at 0.0727 0.0933>,<at 0.6644 0.015>]}
      }})
    Lifetime=&UniformRangedFloat(Minimum=0.45, Maximum=1.5)
    Module.Write Color=true
    Module.Write Lifetime=true
    Module.Write Mass=true
    Module.Write Material Random=true
    Module.Write Position=false
    Module.Write Scale=false
    Module.Write SpriteRotation=true
    Module.Write SpriteSize=true
    Position=&SimulationPosition()
    Sprite Rotation=&UniformRangedFloat(Minimum=0, Maximum=360)
    Sprite Size=&Vector2DFromFloat(Value=&UniformRangedFloat(Minimum=3, Maximum=5))
   )
   &AccelerationForce(Acceleration=&UniformRangedVector(Minimum=<0 0 5>, Maximum=<0 0 40>))
   &AddVelocity(Scale Added Velocity=<0.8 0.8 1>, Velocity=&UniformRangedVector(Minimum=<-20 -20 5>, Maximum=<20 20 100>))
   &ApplyInitialForces()
  ]
  particleUpdateScript [
   &ParticleState()
   &ScaleColor(
    Scale Alpha=&FloatFromCurve(
     FloatCurve=DataInterface() {propertyValues {Curve {keys [
         {arriveTangent=-3.1757, interpMode=RCIM_Cubic, leaveTangent=-3.1757, tangentMode=RCTM_User, time=0.5, value=1.5}
         {arriveTangent=-1.5121, interpMode=RCIM_Cubic, leaveTangent=-1.5121, tangentMode=RCTM_User, time=1.0011, value=0.002}
        ]}}}
     Scale Curve=0.85
    )
    Scale RGB=&Multiply_VectorByFloat(Vector=<0.3 0.3 0.3>, Float=&RandomRangeFloat(Minimum=0.65, Maximum=0.15))
    ScaleA=true
    ScaleRGB=true
   )
   &Drag(Drag=3, Use Linear Drag=true, Use Rotational Drag=false)
   &SolveForcesAndVelocity(Clamp Velocity=false, Limit Acceleration=false)
   &ScaleSpriteSize(Scale Factor=&Vector2DFromFloat(Value=&FloatFromCurve(
      FloatCurve=DataInterface() {propertyValues {Curve {keys [
          <at 0 0>
          {arriveTangent=1.5202, interpMode=RCIM_Cubic, leaveTangent=1.5202, tangentMode=RCTM_User, time=0.2, value=1.1969}
          <at 1.0011 3.4809 RCIM_Cubic>
         ]}}}
      Scale Curve=2
     )))
   /SetVariables(Particles.SubImageIndex=&FloatFromCurve(FloatCurve=DataInterface() {propertyValues {Curve {keys [{arriveTangent=29.8623, interpMode=RCIM_Cubic, leaveTangent=29.8623, tangentMode=RCTM_User, time=0, value=0}, <at 1.0011 22.1689 RCIM_Cubic>]}}})) {isSetParameters=true, module=SetVariables_#1}
   &DynamicMaterialParameters(
    Index 0 Param 1=&FloatFromCurve(FloatCurve=DataInterface() {propertyValues {Curve {keys [<at -0.0046 1.0113 RCIM_Cubic>,<at 1.0011 0.002>]}}})
    Index 0 Param 4=&FloatFromCurve(FloatCurve=DataInterface() {propertyValues {Curve {keys [<at 0 0 RCIM_Cubic>,<at 1 0.5>]}}})
    Module.Param0WriteEnabled=true
    Module.Param1WriteEnabled=false
    Module.Param2WriteEnabled=false
    Module.Param3WriteEnabled=true
   ) {module=DynamicMatParams}
   &CameraOffset(Camera Offset Amount=20)
  ]
  rendererClasses [&NiagaraSpriteRendererProperties]
  renderers [{rendererClass=&NiagaraSpriteRendererProperties}]
 }
 SmokeTrail001 {
  emitterUpdateScript [
   &EmitterScalability(
    Module.bApplySpawnCountScale=false
    Module.bApplySpawnCountScaleByDistanceFraction=false
    Module.bUseMaxDistance=false
    Module.bUseMinDistance=false
   ) {enabled=false}
   &EmitterState(Life Cycle Mode=::NewEnumerator1, Loop Behavior=::NewEnumerator1, Loop Duration=1.1, UseLoopDelay=false)
   &EmitterLifeCycle() {enabled=false}
   &SpawnPerUnit(Module.Use Max Movement Threshold=true, Module.Use Movement Tolerance=false, Spawn Group=1, Spawn Per Unit=2, Use Spawn Probability=false) {enabled=false}
  ]
  particleSpawnScript [
   &InitializeRibbon(
    Color=<rgba 0 0.125 0.1875 0.4>
    Lifetime=1.5
    Module.Write Color=true
    Module.Write Lifetime=true
    Module.Write Mass=true
    Module.Write Material Random=true
    Module.Write Position=true
    Module.Write RibbonTwist=true
    Module.Write RibbonWidth=true
    Position:Vector3f=&Add_Vector(A=/Niagara/DynamicInputs/Helpers/SimulationPosition(), B=<0 0 0>)
    Ribbon Twist=1
    Ribbon Width=20
   )
   &AccelerationForce(Acceleration=<0 0 10>) {module=AccelerationForce001}
   &AddVelocity(Scale Added Velocity=<2 2 1>, Velocity=&RandomRangeVector(Minimum=<10 20 0>, Maximum=<-10 -10 0>))
   &ApplyInitialForces()
  ]
  particleUpdateScript [
   &ParticleState()
   &AccelerationForce(Acceleration=&RandomRangeVector(Minimum=<0 0 50>, Maximum=<0 0 100>))
   &SolveForcesAndVelocity(Clamp Velocity=false, Limit Acceleration=false)
   &Color(Color=$Particles.Initial.Color)
   /SetVariables(Particles.RibbonWidth=&FloatFromCurve(FloatCurve=DataInterface() {propertyValues {Curve {keys [<at 0 1>,<at 1 0>]}}}, Scale Curve=6)) {enabled=false, isSetParameters=true}
   &DynamicMaterialParameters(
    Index 0 Param 1=&FloatFromCurve(FloatCurve=DataInterface() {propertyValues {Curve {keys [<at -0.0046 1.0113>,<at 0.5201 0.2342>,<at 0.7 0.1>]}}})
    Index 0 Param 2=0.1
    Index 0 Param 3=-3
    Index 0 Param 4=&FloatFromCurve(FloatCurve=DataInterface() {propertyValues {Curve {keys [<at -0.0046 1.0113>,<at 1 0>]}}})
    Module.Param0WriteEnabled=true
    Module.Param1WriteEnabled=true
    Module.Param2WriteEnabled=true
    Module.Param3WriteEnabled=true
   )
   &ScaleColor(
    Scale Alpha=1.5
    Scale RGB=&VectorFromFloat(Value=&FloatFromCurve(FloatCurve=DataInterface() {propertyValues {Curve {keys [
          {arriveTangent=-0.0519, interpMode=RCIM_Cubic, leaveTangent=-0.0519, tangentMode=RCTM_User, time=0, value=1}
          {arriveTangent=-0.9379, interpMode=RCIM_Cubic, leaveTangent=-0.9379, tangentMode=RCTM_User, time=0.2578, value=0.2076}
          {arriveTangent=-0.2078, interpMode=RCIM_Cubic, leaveTangent=-0.2078, tangentMode=RCTM_User, time=1, value=0}
         ]}}}))
    ScaleA=true
    ScaleRGB=true
   )
  ]
  rendererClasses [&NiagaraRibbonRendererProperties]
  renderers [{rendererClass=&NiagaraRibbonRendererProperties}]
 }
 SmokeTrail_01 {
  emitterUpdateScript [
   &EmitterScalability(
    Module.bApplySpawnCountScale=false
    Module.bApplySpawnCountScaleByDistanceFraction=false
    Module.bUseMaxDistance=false
    Module.bUseMinDistance=false
   ) {enabled=false}
   &EmitterState(Life Cycle Mode=::NewEnumerator1, Loop Behavior=::NewEnumerator1, Loop Duration=1.7, UseLoopDelay=false)
   &EmitterLifeCycle() {enabled=false}
   &SpawnRate(SpawnRate=10, Use Spawn Probability=false) {enabled=false}
  ]
  particleSpawnScript [
   &InitializeRibbon(
    Color=<rgba 0 0.6667 1 0.4>
    Lifetime=0.6
    Module.Write Color=true
    Module.Write Lifetime=true
    Module.Write Mass=true
    Module.Write Material Random=true
    Module.Write Position=true
    Module.Write RibbonTwist=true
    Module.Write RibbonWidth=true
    Position:Vector3f=&Add_Vector(A=&SimulationPosition(), B=<0 0 0>)
    Ribbon Twist=1
    Ribbon Width=18
   )
   &AccelerationForce(Acceleration=<0 0 50>) {module=AccelerationForce001}
   &AddVelocity(Velocity=&RandomRangeVector(Minimum=<10 20 10>, Maximum=<-10 0 -10>))
   &ApplyInitialForces()
  ]
  particleUpdateScript [
   &ParticleState()
   &AccelerationForce(Acceleration=&RandomRangeVector(Minimum=<0 0 50>, Maximum=<0 0 100>))
   &SolveForcesAndVelocity(Clamp Velocity=false, Limit Acceleration=false)
   &Color(Color=$Particles.Initial.Color)
   &DynamicMaterialParameters(
    Index 0 Param 1=&FloatFromCurve(FloatCurve=DataInterface() {propertyValues {Curve {keys [<at -0.0046 1.0113>,<at 0.5201 0.2342>,<at 0.7 0.1>]}}})
    Index 0 Param 2=0.1
    Index 0 Param 3=-3
    Index 0 Param 4=&FloatFromCurve(FloatCurve=DataInterface() {propertyValues {Curve {keys [<at -0.0046 1.0113>,<at 1 0>]}}})
    Module.Param0WriteEnabled=true
    Module.Param1WriteEnabled=true
    Module.Param2WriteEnabled=true
    Module.Param3WriteEnabled=true
   )
   &ScaleColor(
    Scale Alpha=1.5
    Scale RGB=&VectorFromFloat(Value=&FloatFromCurve(FloatCurve=DataInterface() {propertyValues {Curve {keys [
          {arriveTangent=-0.0519, interpMode=RCIM_Cubic, leaveTangent=-0.0519, tangentMode=RCTM_User, time=0, value=1}
          {arriveTangent=-0.9379, interpMode=RCIM_Cubic, leaveTangent=-0.9379, tangentMode=RCTM_User, time=0.2578, value=0.2076}
          {arriveTangent=-0.2078, interpMode=RCIM_Cubic, leaveTangent=-0.2078, tangentMode=RCTM_User, time=1, value=0}
         ]}}}))
    ScaleA=true
    ScaleRGB=true
   )
  ]
  rendererClasses [&NiagaraRibbonRendererProperties]
  renderers [{rendererClass=&NiagaraRibbonRendererProperties}]
 }
}
summary {
 systemName=NS_Emote_ManAndMonster_Beaker_01
 emitters [
  {emitterName=Beaker, rendererClasses [&NiagaraMeshRendererProperties]}
  {emitterName=GlassCrack, rendererClasses [&NiagaraMeshRendererProperties]}
  {emitterName=Dust, rendererClasses [&NiagaraSpriteRendererProperties]}
  {emitterName=SmokeTrail_01, rendererClasses [&NiagaraRibbonRendererProperties]}
  {emitterName=SmokeTrail001, rendererClasses [&NiagaraRibbonRendererProperties]}
  {emitterName=Smoke, rendererClasses [&NiagaraSpriteRendererProperties]}
  {emitterName=Goop, rendererClasses [&NiagaraMeshRendererProperties]}
  {emitterName=Beaker_Frontend, rendererClasses [&NiagaraMeshRendererProperties]}
 ]
 userVariables [
  User.bIsFrontEnd {type=NiagaraBool}
  User.bIsFrontEndPreview {type=NiagaraBool}
  User.IsSprinting {type=NiagaraBool}
  User.SkeletalMesh {
   type=NiagaraDataInterfaceSkeletalMesh
   defaultValue {
    struct=NiagaraExt_VariableValue_DataInterface
    value {dataInterface=&self:NiagaraDataInterfaceSkeletalMesh_0}
   }
  }
 ]
}
