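# Niagara sprite effect for the ManAndMonster emote: blue bubbles that rise and pop, with smoke.
# Bubbles spawn with falling probability over each loop, drift upward, swell before expiring and emit death events.
# Pops01 and Pops02 have spawn bursts disabled, suggesting those events drive them, though no handlers appear here.
# Smoke combines a burst and steady spawning with animated sub-images and dynamic material parameters, while Spike adds brief, bright radial streaks.
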NiagaraSystem /BRCosmetics/Athena/Effects/Particles/Emotes/NS_Emote_ManAndMonster/NS_Emote_ManAndMonster_Bubble_01
dependencies {
 usedDataInterfaces [&NiagaraDataInterfaceColorCurve, &NiagaraDataInterfaceCurve]
 usedDynamicInputs []
 usedModules [&SystemState]
}
emitters {
 Bubbles {
  emitterUpdateScript [
   &EmitterState(Life Cycle Mode=::NewEnumerator1, Loop Behavior=::NewEnumerator1, Loop Delay=0.05, Loop Duration=1.6, UseLoopDelay=true)
   &SpawnRate(
    Spawn Probability=&FloatFromCurve(
     FloatCurve=DataInterface() {propertyValues {Curve {keys [{arriveTangent=-2.786, interpMode=RCIM_Cubic, leaveTangent=-2.786, tangentMode=RCTM_User, time=-0.0046, value=1.0113}, <at 1.0011 0.002>]}}}
     CurveIndex=$Emitter.NormalizedLoopAge
    )
    SpawnRate=12
    Use Spawn Probability=true
   )
  ]
  particleSpawnScript [
   /Niagara/Modules/Spawn/Initialization/InitializeParticle(
    Color=&RandomRangeLinearColor(Minimum=<rgba 0.1 0.7 1 1>, Maximum=<rgba 0.2 0.6 1 1>)
    Lifetime=&RandomRangeFloat(Minimum=0.4, Maximum=0.8)
    Mesh Scale=<1.4 1.4 1.4>
    Module.Write Color=true
    Module.Write Lifetime=true
    Module.Write Mass=true
    Module.Write Material Random=true
    Module.Write Position=false
    Module.Write Scale=false
    Module.Write SpriteRotation=false
    Module.Write SpriteSize=true
    Position=&SimulationPosition()
    Sprite Size=&Vector2DFromFloat(Value=&UniformRangedFloat(Minimum=1.2, Maximum=3.5))
   )
   &AddVelocity(Velocity=&UniformRangedVector(Minimum=<-5 -5 10>, Maximum=<5 5 40>))
   &CylinderLocation(
    Cylinder Height=0
    Cylinder Origin:NiagaraPosition=$Engine.Owner.Position
    Cylinder Radius=3
    Lathe Profile=DataInterface() {propertyValues {Curve {keys []}}}
    MaskSpawnGroup=false
    Module.Enable Lathe Profile=false
    Module.Override Randomness=false
    Module.Surface Only=false
   )
   /SetVariables(Particles.RandomPosNeg:NiagaraFloat=&UniformRangedFloat(Minimum=-1, Maximum=1)) {isSetParameters=true}
  ]
  particleUpdateScript [
   &ParticleState()
   &ScaleColor(Scale RGB=&VectorFromFloat(Value=20), ScaleA=true, ScaleRGB=true)
   &AccelerationForce(Acceleration=&UniformRangedVector(Minimum=<-30 -30 30>, Maximum=<30 30 1>))
   &Drag(Use Linear Drag=true, Use Rotational Drag=false)
   &SolveForcesAndVelocity(Clamp Velocity=false, Limit Acceleration=false)
   &ScaleSpriteSize(Scale Factor=&Vector2DFromFloat(Value=&FloatFromCurve(FloatCurve=DataInterface() {propertyValues {Curve {keys [
          <at 0 0 RCIM_Cubic>
          {arriveTangent=0.002, interpMode=RCIM_Cubic, leaveTangent=-0.0489, tangentMode=RCTM_Break, time=0.0313, value=0.9069}
          {arriveTangent=-0.077, interpMode=RCIM_Cubic, leaveTangent=-0.077, tangentMode=RCTM_User, time=0.85, value=0.8888}
          {arriveTangent=4.9648, interpMode=RCIM_Cubic, leaveTangent=4.9648, tangentMode=RCTM_User, time=1, value=1.7}
         ]}}})))
   /Niagara/Modules/Update/Orientation/V2/UpdateMeshOrientation(
    Rotation Rate:NiagaraFloat=&Multiply_Float(A=0.035, B=$Particles.RandomPosNeg)
    Rotation Vector:Vector3f=&VectorFromFloat(Value=1)
   )
   &GenerateDeathEvent(Gap Correction Amount=1)
  ]
  rendererClasses [&NiagaraSpriteRendererProperties]
  renderers [{rendererClass=&NiagaraSpriteRendererProperties}]
 }
 Pops01 {
  emitterUpdateScript [
   &EmitterState()
   &SpawnBurst_Instantaneous(Spawn Count=1, Use Spawn Probability=false) {enabled=false}
  ]
  particleSpawnScript [
   &InitializeParticle(
    Color=&ColorFromCurve(ColorCurve=DataInterface() {propertyValues {
       AlphaCurve {keys [<at 0 1 RCIM_Cubic>,<at 1 0 RCIM_Cubic>]}
       BlueCurve {keys [<at 0 1 RCIM_Cubic>]}
       GreenCurve {keys [<at 0 0.76 RCIM_Cubic>]}
       RedCurve {keys [<at 0 0.4 RCIM_Cubic>]}
      }})
    Lifetime=&RandomRangeFloat(Minimum=0.15, Maximum=0.1)
    Mesh Scale=<1.4 1.4 1.4>
    Module.Write Color=true
    Module.Write Lifetime=true
    Module.Write Mass=true
    Module.Write Material Random=true
    Module.Write Position=false
    Module.Write Scale=false
    Module.Write SpriteRotation=true
    Module.Write SpriteSize=true
    Position=&SimulationPosition()
    Sprite Rotation=&RandomRangeFloat(Minimum=-1, Maximum=1)
    Sprite Size=&Vector2DFromFloat(Value=&UniformRangedFloat(Minimum=7.5, Maximum=8))
   )
   /SetVariables(Particles.SubImageIndex=&RandomRangeFloat(Minimum=0, Maximum=3)) {isSetParameters=true, module=SetVariables_#1}
  ]
  particleUpdateScript [
   &ParticleState()
   &AccelerationForce(Acceleration=&UniformRangedVector(Minimum=<0 0 -50>, Maximum=<0 0 -80>))
   &SolveForcesAndVelocity(Clamp Velocity=false, Limit Acceleration=false)
   &ScaleSpriteSize(Scale Factor=&Vector2DFromFloat(Value=&FloatFromCurve(
      FloatCurve=DataInterface() {propertyValues {Curve {keys [
          {arriveTangent=8.9659, interpMode=RCIM_Cubic, leaveTangent=8.9659, tangentMode=RCTM_User, time=0, value=0}
          {arriveTangent=0.002, interpMode=RCIM_Cubic, leaveTangent=0.6693, tangentMode=RCTM_Break, time=0.1, value=0.9069}
          {arriveTangent=0.4183, interpMode=RCIM_Cubic, leaveTangent=0.4183, tangentMode=RCTM_User, time=1, value=1.2}
         ]}}}
      Scale Curve=1.2
     )))
   &ScaleColor(
    Scale RGB=&VectorFromFloat(Value=&RandomRangeFloat(Minimum=0.5, Maximum=0.7))
    ScaleA=true
    ScaleRGB=true
   )
  ]
  rendererClasses [&NiagaraSpriteRendererProperties]
  renderers [{rendererClass=&NiagaraSpriteRendererProperties}]
 }
 Pops02 {
  emitterUpdateScript [
   &EmitterState()
   &SpawnBurst_Instantaneous(Spawn Count=1, Use Spawn Probability=false) {enabled=false}
  ]
  particleSpawnScript [
   &InitializeParticle(
    Color=&RandomRangeLinearColor(Minimum=<rgba 0 0.2031 0.6094 1>, Maximum=<rgba 0 0.1667 1 2>)
    Lifetime=&RandomRangeFloat(Minimum=0.35, Maximum=0.25)
    Mesh Scale=<1.4 1.4 1.4>
    Module.Write Color=true
    Module.Write Lifetime=true
    Module.Write Mass=true
    Module.Write Material Random=true
    Module.Write Position=false
    Module.Write Scale=false
    Module.Write SpriteRotation=true
    Module.Write SpriteSize=true
    Position=&SimulationPosition()
    Sprite Rotation=&RandomRangeFloat(Minimum=-1, Maximum=1)
    Sprite Size=&Vector2DFromFloat(Value=&UniformRangedFloat(Minimum=6.5, Maximum=5))
   )
   /SetVariables(Particles.SubImageIndex=&RandomRangeFloat(Minimum=0, Maximum=3)) {isSetParameters=true, module=SetVariables_#1}
  ]
  particleUpdateScript [
   &ParticleState()
   &AccelerationForce(Acceleration=&UniformRangedVector(Minimum=<0 0 -150>, Maximum=<0 0 -80>))
   &SolveForcesAndVelocity(Clamp Velocity=false, Limit Acceleration=false)
   &ScaleSpriteSize(Scale Factor=&Vector2DFromFloat(Value=&FloatFromCurve(FloatCurve=DataInterface() {propertyValues {Curve {keys [
          {arriveTangent=8.9659, interpMode=RCIM_Cubic, leaveTangent=8.9659, tangentMode=RCTM_User, time=0, value=0}
          {arriveTangent=0.002, interpMode=RCIM_Cubic, leaveTangent=0.6693, tangentMode=RCTM_Break, time=0.1, value=0.9069}
          {arriveTangent=0.4183, interpMode=RCIM_Cubic, leaveTangent=0.4183, tangentMode=RCTM_User, time=1, value=1.2}
         ]}}})))
   &DynamicMaterialParameters(
    Index 0 Param 1=&FloatFromCurve(FloatCurve=DataInterface() {propertyValues {Curve {keys [<at 0 1>,<at 1 0>]}}})
    Index 0 Param 2=1
    Index 0 Param 3=1
    Index 0 Param 4=1
    Module.Param0WriteEnabled=true
    Module.Param1WriteEnabled=true
    Module.Param2WriteEnabled=true
    Module.Param3WriteEnabled=true
   )
   &ScaleColor(Scale RGB=&VectorFromFloat(Value=20), ScaleA=true, ScaleRGB=true)
  ]
  rendererClasses [&NiagaraSpriteRendererProperties]
  renderers [{rendererClass=&NiagaraSpriteRendererProperties}]
 }
 Smoke {
  emitterUpdateScript [
   &EmitterState(Life Cycle Mode=::NewEnumerator1, Loop Behavior=::NewEnumerator1, Loop Duration=0.5, UseLoopDelay=false)
   &SpawnBurst_Instantaneous(Spawn Count=35, Spawn Time=0.05, Use Spawn Probability=false) {module=SpawnBurst_Instantaneous001}
   &SpawnRate(SpawnRate=100, Use Spawn Probability=false)
  ]
  particleSpawnScript [
   &InitializeParticle(
    Color=&ColorFromCurve(ColorCurve=DataInterface() {propertyValues {
       AlphaCurve {keys [<at -0.0445 0.1>]}
       BlueCurve {keys [<at 0 3 RCIM_Cubic>,<at 0.0727 0.7>,<at 0.6644 0.3>]}
       GreenCurve {keys [<at 0 1.7 RCIM_Cubic>,<at 0.0727 0.245>,<at 0.6644 0.015>]}
       RedCurve {keys [<at 0 0.4 RCIM_Cubic>,<at 0.0727 0.0933>,<at 0.6644 0.015>]}
      }})
    Lifetime=&UniformRangedFloat(Minimum=0.45, Maximum=0.9)
    Module.Write Color=true
    Module.Write Lifetime=true
    Module.Write Mass=true
    Module.Write Material Random=true
    Module.Write Position=false
    Module.Write Scale=false
    Module.Write SpriteRotation=true
    Module.Write SpriteSize=true
    Position=&SimulationPosition()
    Sprite Rotation=&UniformRangedFloat(Minimum=0, Maximum=360)
    Sprite Size=&Vector2DFromFloat(Value=&UniformRangedFloat(Minimum=3, Maximum=5))
   )
   &AccelerationForce(Acceleration=&UniformRangedVector(Minimum=<0 0 5>, Maximum=<0 0 40>))
   &AddVelocity(Scale Added Velocity=<0.8 0.8 1>, Velocity=&UniformRangedVector(Minimum=<-20 -20 5>, Maximum=<20 20 20>))
   &ApplyInitialForces()
  ]
  particleUpdateScript [
   &ParticleState()
   &ScaleColor(
    Scale Alpha=&FloatFromCurve(
     FloatCurve=DataInterface() {propertyValues {Curve {keys [
         {arriveTangent=-3.1757, interpMode=RCIM_Cubic, leaveTangent=-3.1757, tangentMode=RCTM_User, time=0.5, value=1.5}
         {arriveTangent=-1.5121, interpMode=RCIM_Cubic, leaveTangent=-1.5121, tangentMode=RCTM_User, time=1.0011, value=0.002}
        ]}}}
     Scale Curve=0.85
    )
    Scale RGB=&Multiply_VectorByFloat(Vector=<0.3 0.3 0.3>, Float=&RandomRangeFloat(Minimum=0.65, Maximum=0.15))
    ScaleA=true
    ScaleRGB=true
   )
   &Drag(Drag=3, Use Linear Drag=true, Use Rotational Drag=false)
   &SolveForcesAndVelocity(Clamp Velocity=false, Limit Acceleration=false)
   &ScaleSpriteSize(Scale Factor=&Vector2DFromFloat(Value=&FloatFromCurve(FloatCurve=DataInterface() {propertyValues {Curve {keys [
          <at 0 0>
          {arriveTangent=1.5202, interpMode=RCIM_Cubic, leaveTangent=1.5202, tangentMode=RCTM_User, time=0.2, value=1.1969}
          <at 1.0011 3.4809 RCIM_Cubic>
         ]}}})))
   /SetVariables(Particles.SubImageIndex=&FloatFromCurve(FloatCurve=DataInterface() {propertyValues {Curve {keys [{arriveTangent=29.8623, interpMode=RCIM_Cubic, leaveTangent=29.8623, tangentMode=RCTM_User, time=0, value=0}, <at 1.0011 22.1689 RCIM_Cubic>]}}})) {isSetParameters=true}
   &DynamicMaterialParameters(
    Index 0 Param 1=&FloatFromCurve(FloatCurve=DataInterface() {propertyValues {Curve {keys [<at -0.0046 1.0113 RCIM_Cubic>,<at 1.0011 0.002>]}}})
    Index 0 Param 4=&FloatFromCurve(FloatCurve=DataInterface() {propertyValues {Curve {keys [<at 0 0 RCIM_Cubic>,<at 1 0.5>]}}})
    Module.Param0WriteEnabled=true
    Module.Param1WriteEnabled=false
    Module.Param2WriteEnabled=false
    Module.Param3WriteEnabled=true
   ) {module=DynamicMatParams}
   &CameraOffset(Camera Offset Amount=20)
  ]
  rendererClasses [&NiagaraSpriteRendererProperties]
  renderers [{rendererClass=&NiagaraSpriteRendererProperties}]
 }
 Spike {
  emitterUpdateScript [
   &EmitterState()
   &SpawnBurst_Instantaneous(Use Spawn Probability=false)
  ]
  particleSpawnScript [
   &InitializeParticle(
    Color=<rgba 0.125 0.4167 1 2>
    Lifetime=&UniformRangedFloat(Minimum=0.1, Maximum=0.13)
    Module.Write Color=true
    Module.Write Lifetime=true
    Module.Write Mass=true
    Module.Write Material Random=true
    Module.Write Position=false
    Module.Write Scale=false
    Module.Write SpriteRotation=false
    Module.Write SpriteSize=true
    Position=&SimulationPosition()
    Sprite Size=&UniformRangedVector2D(Minimum=<0.1 0.9>, Maximum=<0.27 0.4>)
   )
   &AddVelocityFromPoint(Module.Use Falloff Distance=false, Velocity Strength=60)
  ]
  particleUpdateScript [
   &ParticleState()
   &SolveForcesAndVelocity(Clamp Velocity=false, Limit Acceleration=false)
   &ScaleColor(
    Scale Alpha=&FloatFromCurve(FloatCurve=DataInterface() {propertyValues {Curve {keys [<at 0 2>,<at 1 0>]}}})
    Scale RGB=&VectorFromFloat(Value=&FloatFromCurve(FloatCurve=DataInterface() {propertyValues {Curve {keys [<at 0 40>,<at 0.5 1>]}}}))
    ScaleA=true
    ScaleRGB=true
   )
   &ScaleSpriteSize(Scale Factor=&Vector2DFromFloat(Value=&FloatFromCurve(
      FloatCurve=DataInterface() {propertyValues {Curve {keys [{arriveTangent=2.4508, interpMode=RCIM_Cubic, leaveTangent=2.4508, tangentMode=RCTM_User, time=0, value=0.4}, <at 1 1.5 RCIM_Cubic>]}}}
      Scale Curve=4
     )))
  ]
  rendererClasses [&NiagaraSpriteRendererProperties]
  renderers [{rendererClass=&NiagaraSpriteRendererProperties}]
 }
}
summary {
 systemName=NS_Emote_ManAndMonster_Bubble_01
 emitters [
  {emitterName=Pops01, rendererClasses [&NiagaraSpriteRendererProperties]}
  {emitterName=Bubbles, rendererClasses [&NiagaraSpriteRendererProperties]}
  {emitterName=Pops02, rendererClasses [&NiagaraSpriteRendererProperties]}
  {emitterName=Spike, rendererClasses [&NiagaraSpriteRendererProperties]}
  {emitterName=Smoke, rendererClasses [&NiagaraSpriteRendererProperties]}
 ]
 userVariables []
}
