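# Niagara system for the ManAndMonster emote, wrapping a character in glowing blue bubbles and smoke.
# Bubbles, SmokeHeavy puffs, ribbon trails and bone-following Initial_Smoke sample the User.SkeletalMesh input through spawn groups, while a delayed SmokeHeavy_Big billow fills a cylinder volume.
# Bubbles swell late in life and emit death events; Pops01, Pops02 and Spike provide short pop sprites, though no event handlers are shown.
# Smoke puffs animate sub-image frames and dynamic material parameters over life.
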NiagaraSystem /BRCosmetics/Athena/Effects/Particles/Emotes/NS_Emote_ManAndMonster/NS_Emote_ManAndMonster_Bubble_02
dependencies {
 usedDataInterfaces [&NiagaraDataInterfaceColorCurve, &NiagaraDataInterfaceCurve, &NiagaraDataInterfaceVector2DCurve]
 usedDynamicInputs []
 usedModules [&SystemState]
}
emitters {
 Bubbles {
  emitterUpdateScript [
   &EmitterState(Life Cycle Mode=::NewEnumerator1, Loop Behavior=::NewEnumerator1, Loop Duration=2.6, UseLoopDelay=false)
   &SpawnRate(SpawnRate=18, Use Spawn Probability=true)
   &SpawnBurst_Instantaneous(Spawn Count=5, Spawn Time=0.15, Use Spawn Probability=false)
  ]
  particleSpawnScript [
   /Niagara/Modules/Spawn/Initialization/InitializeParticle(
    Color=&RandomRangeLinearColor(Minimum=<rgba 0.1 0.85 1 1>, Maximum=<rgba 0.2 0.6 1 1>)
    Lifetime=&RandomRangeFloat(Minimum=0.4, Maximum=0.8)
    Mesh Scale=<1.4 1.4 1.4>
    Module.Write Color=true
    Module.Write Lifetime=true
    Module.Write Mass=true
    Module.Write Material Random=true
    Module.Write Position=false
    Module.Write Scale=false
    Module.Write SpriteRotation=false
    Module.Write SpriteSize=true
    Position=&SimulationPosition()
    Sprite Size=&Vector2DFromFloat(Value=&UniformRangedFloat(Minimum=2, Maximum=5))
   )
   &AddVelocity(Velocity=&UniformRangedVector(Minimum=<-5 -5 10>, Maximum=<5 5 40>))
   &CylinderLocation(
    Cylinder Height=5
    Cylinder Origin:NiagaraPosition=$Engine.Owner.Position
    Cylinder Radius=2
    Lathe Profile=DataInterface() {propertyValues {Curve {keys []}}}
    MaskSpawnGroup=false
    Module.Enable Lathe Profile=false
    Module.Override Randomness=false
    Module.Surface Only=false
   ) {enabled=false}
   &FN_SetSpawnGroup_v2(Number of Spawn Groups=7)
   &SkeletalMeshLocation(
    Bone / Socket Index=&SpawnGroup()
    Bone Sampling Mode=::NewEnumerator4
    Inherit Velocity=false
    MaskSpawnGroup=false
    Sampled Position Offset=&RandomRangeVector(Minimum=<5 5 5>, Maximum=<-5 -5 -5>)
    Skeletal Mesh=$User.SkeletalMesh
   )
  ]
  particleUpdateScript [
   &ParticleState()
   &ScaleColor(Scale Alpha=2, Scale RGB=&VectorFromFloat(Value=30), ScaleA=true, ScaleRGB=true)
   &AccelerationForce(Acceleration=&UniformRangedVector(Minimum=<-30 -30 30>, Maximum=<30 30 1>))
   &Drag(Use Linear Drag=true, Use Rotational Drag=false)
   &SolveForcesAndVelocity(Clamp Velocity=false, Limit Acceleration=false)
   &ScaleSpriteSize(Scale Factor=&Vector2DFromFloat(Value=&FloatFromCurve(FloatCurve=DataInterface() {propertyValues {Curve {keys [
          <at 0 0 RCIM_Cubic>
          {arriveTangent=0.002, interpMode=RCIM_Cubic, leaveTangent=-0.0489, tangentMode=RCTM_Break, time=0.0313, value=0.9069}
          {arriveTangent=-0.077, interpMode=RCIM_Cubic, leaveTangent=-0.077, tangentMode=RCTM_User, time=0.85, value=0.8888}
          {arriveTangent=4.9648, interpMode=RCIM_Cubic, leaveTangent=4.9648, tangentMode=RCTM_User, time=1, value=1.7}
         ]}}})))
   &GenerateDeathEvent(Gap Correction Amount=1)
  ]
  rendererClasses [&NiagaraSpriteRendererProperties]
  renderers [{rendererClass=&NiagaraSpriteRendererProperties}]
 }
 Initial_Smoke {
  emitterUpdateScript [
   &EmitterState(Life Cycle Mode=::NewEnumerator1, Loop Behavior=::NewEnumerator1, Loop Duration=0.75, UseLoopDelay=false)
   &SpawnRate(
    SpawnRate=&FloatFromCurve(
     FloatCurve=DataInterface() {propertyValues {Curve {keys [<at 0 1 RCIM_Cubic>, {arriveTangent=-1.85, interpMode=RCIM_Cubic, leaveTangent=-1.85, tangentMode=RCTM_User, time=1, value=0}]}}}
     CurveIndex=$Emitter.NormalizedLoopAge
     Scale Curve=20
    )
    Use Spawn Probability=false
   )
  ]
  particleSpawnScript [
   &FN_SetSpawnGroup_v2(Number of Spawn Groups=12)
   &InitializeParticle(
    Color=&MultiplyLinearColor(
     Linear Color=&UniformRangedLinearColor(Minimum=<rgba 0 0.764 1 1>, Maximum=<rgba 0.2 1.25 2 1>, Color Channel Mode=::NewEnumerator1)
     Scale Factor=<1 1 1 1>
    )
    Lifetime=&FloatFromCurve(FloatCurve=DataInterface() {propertyValues {Curve {keys [<at 0 1.25 RCIM_Cubic>,<at 1 0.2>]}}}, CurveIndex=$Emitter.NormalizedLoopAge)
    Module.Write Color=true
    Module.Write Lifetime=true
    Module.Write Mass=false
    Module.Write Material Random=true
    Module.Write Position=false
    Module.Write Scale=false
    Module.Write SpriteRotation=true
    Module.Write SpriteSize=true
    Sprite Size=&UniformRangedVector2D(Minimum=<15 40>, Maximum=<35 70>)
   )
   &SystemLocation(MaskSpawnGroup=false, Offset=<0 0 100>) {enabled=false}
   &CameraOffset(Camera Offset Amount=&UniformAOrBFloat(A=5, B=16))
   &DynamicMaterialParameters(
    Index 0 Param 2=$Particles.MaterialRandom
    Module.Param0WriteEnabled=false
    Module.Param1WriteEnabled=true
    Module.Param2WriteEnabled=false
    Module.Param3WriteEnabled=false
   ) {module=DynamicMaterialParameters001}
   &AddVelocity(Velocity=&MakeVector(X=0, Y=0, Z=&UniformRangedFloat(Minimum=15, Maximum=28))) {enabled=false}
  ]
  particleUpdateScript [
   &ParticleState()
   &ScaleColor(
    Scale Alpha=&FloatFromCurve(FloatCurve=DataInterface() {propertyValues {Curve {keys [
         {arriveTangent=-8.0953, interpMode=RCIM_Cubic, leaveTangent=-8.0952, tangentMode=RCTM_User, time=0, value=1}
         {arriveTangent=-0.8371, interpMode=RCIM_Cubic, leaveTangent=-0.8371, time=0.1955, value=0.0807}
         {arriveTangent=1.1427, interpMode=RCIM_Cubic, leaveTangent=1.1427, time=0.8123, value=0.32}
         {interpMode=RCIM_Cubic, leaveTangent=12.0487, time=1, value=1}
        ]}}})
    ScaleA=true
    ScaleRGB=false
   )
   &SkeletalMeshLocation(
    Bone / Socket Index=$Particles.SpawnGroup
    Bone Sampling Mode=::NewEnumerator4
    Inherit Velocity=false
    MaskSpawnGroup=false
    Sampled Position Offset=<0 0 -6>
    Skeletal Mesh=$User.SkeletalMesh
   )
   &DynamicMaterialParameters(
    Index 0 Param 1=&FloatFromCurve(
     FloatCurve=DataInterface() {propertyValues {Curve {keys [
         {arriveTangent=4.5199, interpMode=RCIM_Cubic, leaveTangent=4.5199, tangentMode=RCTM_User, time=0, value=0}
         {arriveTangent=0.8527, interpMode=RCIM_Cubic, leaveTangent=0.8527, tangentMode=RCTM_User, time=0.2951, value=0.6871}
         {arriveTangent=0.3142, interpMode=RCIM_Cubic, leaveTangent=0.3142, tangentMode=RCTM_User, time=1, value=1}
        ]}}}
     Scale Curve=2
    )
    Module.Param0WriteEnabled=true
    Module.Param1WriteEnabled=false
    Module.Param2WriteEnabled=false
    Module.Param3WriteEnabled=false
   )
   &SolveForcesAndVelocity(Clamp Velocity=false, Limit Acceleration=false) {enabled=false}
   &ScaleSpriteSize(Scale Factor=&Vector2DFromCurve(
     Vector2Curve=DataInterface() {propertyValues {XCurve {keys [<at 0 1 RCIM_Cubic>,<at 1 0.6 RCIM_Cubic>]}, YCurve {keys [<at 0 0.3 RCIM_Cubic>,<at 1 1 RCIM_Cubic>]}}}
     ScaleCurve=<1 1>
    ))
  ]
  rendererClasses [&NiagaraSpriteRendererProperties]
  renderers [{rendererClass=&NiagaraSpriteRendererProperties}]
 }
 Pops01 {
  emitterUpdateScript [
   &EmitterState()
   &SpawnBurst_Instantaneous(Spawn Count=1, Use Spawn Probability=false) {enabled=false}
  ]
  particleSpawnScript [
   &InitializeParticle(
    Color=&ColorFromCurve(ColorCurve=DataInterface() {propertyValues {
       AlphaCurve {keys [<at 0 1 RCIM_Cubic>,<at 1 0 RCIM_Cubic>]}
       BlueCurve {keys [<at 0 1 RCIM_Cubic>]}
       GreenCurve {keys [<at 0 0.7596 RCIM_Cubic>]}
       RedCurve {keys [<at 0 0.4 RCIM_Cubic>]}
      }})
    Lifetime=&RandomRangeFloat(Minimum=0.15, Maximum=0.1)
    Mesh Scale=<1.4 1.4 1.4>
    Module.Write Color=true
    Module.Write Lifetime=true
    Module.Write Mass=true
    Module.Write Material Random=true
    Module.Write Position=false
    Module.Write Scale=false
    Module.Write SpriteRotation=true
    Module.Write SpriteSize=true
    Position=&SimulationPosition()
    Sprite Rotation=&RandomRangeFloat(Minimum=-1, Maximum=1)
    Sprite Size=&Vector2DFromFloat(Value=&UniformRangedFloat(Minimum=7.5, Maximum=8))
   )
   /SetVariables(Particles.SubImageIndex=&RandomRangeFloat(Minimum=0, Maximum=3)) {isSetParameters=true, module=SetVariables_#1}
  ]
  particleUpdateScript [
   &ParticleState()
   &AccelerationForce(Acceleration=&UniformRangedVector(Minimum=<0 0 -50>, Maximum=<0 0 -80>))
   &SolveForcesAndVelocity(Clamp Velocity=false, Limit Acceleration=false)
   &ScaleSpriteSize(Scale Factor=&Vector2DFromFloat(Value=&FloatFromCurve(
      FloatCurve=DataInterface() {propertyValues {Curve {keys [
          {arriveTangent=8.9659, interpMode=RCIM_Cubic, leaveTangent=8.9659, tangentMode=RCTM_User, time=0, value=0}
          {arriveTangent=0.002, interpMode=RCIM_Cubic, leaveTangent=0.6693, tangentMode=RCTM_Break, time=0.1, value=0.9069}
          {arriveTangent=0.4183, interpMode=RCIM_Cubic, leaveTangent=0.4183, tangentMode=RCTM_User, time=1, value=1.2}
         ]}}}
      Scale Curve=1.2
     )))
   &ScaleColor(
    Scale RGB=&VectorFromFloat(Value=&RandomRangeFloat(Minimum=0.5, Maximum=0.7))
    ScaleA=true
    ScaleRGB=true
   )
  ]
  rendererClasses [&NiagaraSpriteRendererProperties]
  renderers [{rendererClass=&NiagaraSpriteRendererProperties}]
 }
 Pops02 {
  emitterUpdateScript [
   &EmitterState()
   &SpawnBurst_Instantaneous(Spawn Count=1, Use Spawn Probability=false) {enabled=false}
  ]
  particleSpawnScript [
   &InitializeParticle(
    Color=&RandomRangeLinearColor(Minimum=<rgba 0 0.3047 0.6094 1>, Maximum=<rgba 0 0.1667 1 2>)
    Lifetime=&RandomRangeFloat(Minimum=0.35, Maximum=0.25)
    Mesh Scale=<1.4 1.4 1.4>
    Module.Write Color=true
    Module.Write Lifetime=true
    Module.Write Mass=true
    Module.Write Material Random=true
    Module.Write Position=false
    Module.Write Scale=false
    Module.Write SpriteRotation=true
    Module.Write SpriteSize=true
    Position=&SimulationPosition()
    Sprite Rotation=&RandomRangeFloat(Minimum=-1, Maximum=1)
    Sprite Size=&Vector2DFromFloat(Value=&UniformRangedFloat(Minimum=6.5, Maximum=5))
   )
   /SetVariables(Particles.SubImageIndex=&RandomRangeFloat(Minimum=0, Maximum=3)) {isSetParameters=true, module=SetVariables_#1}
  ]
  particleUpdateScript [
   &ParticleState()
   &AccelerationForce(Acceleration=&UniformRangedVector(Minimum=<0 0 -150>, Maximum=<0 0 -80>))
   &SolveForcesAndVelocity(Clamp Velocity=false, Limit Acceleration=false)
   &ScaleSpriteSize(Scale Factor=&Vector2DFromFloat(Value=&FloatFromCurve(FloatCurve=DataInterface() {propertyValues {Curve {keys [
          {arriveTangent=8.9659, interpMode=RCIM_Cubic, leaveTangent=8.9659, tangentMode=RCTM_User, time=0, value=0}
          {arriveTangent=0.002, interpMode=RCIM_Cubic, leaveTangent=0.6693, tangentMode=RCTM_Break, time=0.1, value=0.9069}
          {arriveTangent=0.4183, interpMode=RCIM_Cubic, leaveTangent=0.4183, tangentMode=RCTM_User, time=1, value=1.2}
         ]}}})))
   &DynamicMaterialParameters(
    Index 0 Param 1=&FloatFromCurve(FloatCurve=DataInterface() {propertyValues {Curve {keys [<at 0 1>,<at 1 0>]}}})
    Index 0 Param 2=1
    Index 0 Param 3=1
    Index 0 Param 4=1
    Module.Param0WriteEnabled=true
    Module.Param1WriteEnabled=true
    Module.Param2WriteEnabled=true
    Module.Param3WriteEnabled=true
   )
   &ScaleColor(Scale RGB=&VectorFromFloat(Value=20), ScaleA=true, ScaleRGB=true)
  ]
  rendererClasses [&NiagaraSpriteRendererProperties]
  renderers [{rendererClass=&NiagaraSpriteRendererProperties}]
 }
 SmokeHeavy {
  emitterUpdateScript [
   &EmitterState(Life Cycle Mode=::NewEnumerator1, Loop Behavior=::NewEnumerator1, Loop Duration=0.8, UseLoopDelay=true)
   &SpawnRate(
    SpawnRate=&FloatFromCurve(
     FloatCurve=DataInterface() {propertyValues {Curve {keys [<at 0 0>,<at 0.2 0.3>,<at 0.4 1>,<at 5.5013 1>]}}}
     CurveIndex=$Emitter.NormalizedLoopAge
     Scale Curve=300
    )
    Use Spawn Probability=false
   )
  ]
  particleSpawnScript [
   &InitializeParticle(
    Color=&ColorFromCurve(ColorCurve=DataInterface() {propertyValues {
       AlphaCurve {keys [<at -0.0445 0.15>]}
       BlueCurve {keys [<at 0 3 RCIM_Cubic>,<at 0.0727 0.7>,<at 0.6644 0.3>]}
       GreenCurve {keys [<at 0 2.1333 RCIM_Cubic>,<at 0.0727 0.2956>,<at 0.6644 0.015>]}
       RedCurve {keys [<at 0 0.4 RCIM_Cubic>,<at 0.0727 0.0933>,<at 0.6644 0.015>]}
      }})
    Lifetime=&UniformRangedFloat(Minimum=0.85, Maximum=0.5)
    Module.Write Color=true
    Module.Write Lifetime=true
    Module.Write Mass=true
    Module.Write Material Random=true
    Module.Write Position=false
    Module.Write Scale=false
    Module.Write SpriteRotation=true
    Module.Write SpriteSize=true
    Position=&SimulationPosition()
    Sprite Rotation=&UniformRangedFloat(Minimum=0, Maximum=360)
    Sprite Size=&Vector2DFromFloat(Value=&UniformRangedFloat(Minimum=3, Maximum=5.5))
   )
   &AccelerationForce(Acceleration=&UniformRangedVector(Minimum=<0 0 15>, Maximum=<0 0 60>))
   &AddVelocity(Scale Added Velocity=<0.8 0.8 1>, Velocity=&UniformRangedVector(Minimum=<-20 -20 5>, Maximum=<20 20 20>))
   &ApplyInitialForces()
   &FN_SetSpawnGroup_v2(Number of Spawn Groups=7)
   &SkeletalMeshLocation(
    Bone / Socket Index=&SpawnGroup()
    Bone Sampling Mode=::NewEnumerator4
    Inherit Velocity=false
    MaskSpawnGroup=false
    Sampled Position Offset=&RandomRangeVector(Minimum=<4 4 5>, Maximum=<-4 -4 -2>)
    Skeletal Mesh=$User.SkeletalMesh
   )
  ]
  particleUpdateScript [
   &ParticleState()
   &Drag(Drag=3, Use Linear Drag=true, Use Rotational Drag=false)
   &SolveForcesAndVelocity(Clamp Velocity=false, Limit Acceleration=false)
   &ScaleColor(
    Scale Alpha=&FloatFromCurve(
     FloatCurve=DataInterface() {propertyValues {Curve {keys [
         {arriveTangent=-3.1757, interpMode=RCIM_Cubic, leaveTangent=-3.1757, tangentMode=RCTM_User, time=0.5, value=1.5}
         {arriveTangent=-1.5121, interpMode=RCIM_Cubic, leaveTangent=-1.5121, tangentMode=RCTM_User, time=1.0011, value=0.002}
        ]}}}
     Scale Curve=0.85
    )
    Scale RGB=&Multiply_VectorByFloat(Vector=<0.3 0.3 0.3>, Float=&RandomRangeFloat(Minimum=0.6, Maximum=0.1))
    ScaleA=true
    ScaleRGB=true
   )
   &ScaleSpriteSize(Scale Factor=&Vector2DFromFloat(Value=&FloatFromCurve(
      FloatCurve=DataInterface() {propertyValues {Curve {keys [
          <at 0 0>
          {arriveTangent=1.5202, interpMode=RCIM_Cubic, leaveTangent=1.5202, tangentMode=RCTM_User, time=0.2, value=1.2}
          {arriveTangent=3.9459, interpMode=RCIM_Cubic, leaveTangent=3.9459, tangentMode=RCTM_User, time=1.0011, value=3.6}
         ]}}}
      Scale Curve=2.2
     )))
   /SetVariables(Particles.SubImageIndex=&FloatFromCurve(FloatCurve=DataInterface() {propertyValues {Curve {keys [{arriveTangent=29.8623, interpMode=RCIM_Cubic, leaveTangent=29.8623, tangentMode=RCTM_User, time=0, value=0}, <at 1.0011 22.1689 RCIM_Cubic>]}}})) {isSetParameters=true, module=SetVariables_#2}
   &DynamicMaterialParameters(
    Index 0 Param 1=&FloatFromCurve(FloatCurve=DataInterface() {propertyValues {Curve {keys [<at -0.0046 1.0113 RCIM_Cubic>,<at 1.0011 0.002>]}}})
    Index 0 Param 4=&FloatFromCurve(FloatCurve=DataInterface() {propertyValues {Curve {keys [<at 0 0 RCIM_Cubic>,<at 1 0.5>]}}})
    Module.Param0WriteEnabled=true
    Module.Param1WriteEnabled=false
    Module.Param2WriteEnabled=false
    Module.Param3WriteEnabled=true
   ) {module=DynamicMatParams}
   &CameraOffset(Camera Offset Amount=15)
  ]
  rendererClasses [&NiagaraSpriteRendererProperties]
  renderers [{rendererClass=&NiagaraSpriteRendererProperties}]
 }
 SmokeHeavy_Big {
  emitterUpdateScript [
   &EmitterState(Life Cycle Mode=::NewEnumerator1, Loop Behavior=::NewEnumerator1, Loop Delay=1.1, Loop Duration=0.5, UseLoopDelay=true)
   &SpawnBurst_Instantaneous(Spawn Count=30, Spawn Time=0.12, Use Spawn Probability=false) {module=SpawnBurst_Instantaneous001}
   &SpawnRate(
    SpawnRate=&FloatFromCurve(
     FloatCurve=DataInterface() {propertyValues {Curve {keys [<at 0 0>,<at 0.2344 0.7976>,<at 0.5 0.9>,<at 1 1>,<at 5.5013 1>]}}}
     CurveIndex=$Emitter.NormalizedLoopAge
     Scale Curve=150
    )
    Use Spawn Probability=false
   )
  ]
  particleSpawnScript [
   &InitializeParticle(
    Color=&ColorFromCurve(ColorCurve=DataInterface() {propertyValues {
       AlphaCurve {keys [<at -0.0445 0.6>,<at 0.7025 0.4>]}
       BlueCurve {keys [<at 0 3 RCIM_Cubic>,<at 0.0727 0.7>,<at 0.6644 0.2>]}
       GreenCurve {keys [<at 0 1.4833 RCIM_Cubic>,<at 0.0727 0.245>,<at 0.6644 0.01>]}
       RedCurve {keys [<at 0 0.4 RCIM_Cubic>,<at 0.0727 0.0933>,<at 0.6644 0.01>]}
      }})
    Lifetime=&UniformRangedFloat(Minimum=1.05, Maximum=1.6)
    Module.Write Color=true
    Module.Write Lifetime=true
    Module.Write Mass=true
    Module.Write Material Random=true
    Module.Write Position=false
    Module.Write Scale=false
    Module.Write SpriteRotation=true
    Module.Write SpriteSize=true
    Position=&SimulationPosition()
    Sprite Rotation=&UniformRangedFloat(Minimum=-180, Maximum=180)
    Sprite Size=&Vector2DFromFloat(Value=&UniformRangedFloat(Minimum=3, Maximum=7))
   )
   &AccelerationForce(Acceleration=&UniformRangedVector(Minimum=<0 0 15>, Maximum=<0 0 80>))
   &AddVelocity(
    Scale Added Velocity=<0.8 0.8 1>
    Velocity=&UniformRangedVector(Minimum=<-20 -20 50>, Maximum=<20 20 120>)
   )
   &ApplyInitialForces()
   &CylinderLocation(
    Cylinder Height=40
    Cylinder Radius=40
    MaskSpawnGroup=false
    Module.Enable Lathe Profile=false
    Module.Override Randomness=false
    Module.Surface Only=false
    Offset=<0 0 47>
   )
   &FN_SetSpawnGroup_v2(Number of Spawn Groups=14) {enabled=false}
   &SkeletalMeshLocation(
    Bone / Socket Index=&SpawnGroup()
    Bone Sampling Mode=::NewEnumerator4
    Inherit Velocity=false
    MaskSpawnGroup=false
    Sampled Position Offset=&RandomRangeVector(Minimum=<7 7 5>, Maximum=<-7 -7 -5>)
    Skeletal Mesh=$User.SkeletalMesh
   ) {enabled=false}
  ]
  particleUpdateScript [
   &ParticleState()
   &ScaleColor(
    Scale Alpha=&FloatFromCurve(
     FloatCurve=DataInterface() {propertyValues {Curve {keys [
         <at 0 0>
         {arriveTangent=-3.1757, interpMode=RCIM_Cubic, leaveTangent=-3.1757, tangentMode=RCTM_User, time=0.3, value=1}
         {arriveTangent=-1.5121, interpMode=RCIM_Cubic, leaveTangent=-1.5121, tangentMode=RCTM_User, time=1.0011, value=0.002}
        ]}}}
     Scale Curve=&RandomRangeFloat(Minimum=0.85, Maximum=0.95)
    )
    Scale RGB=&Multiply_VectorByFloat(
     Vector=&VectorFromFloat(Value=&FloatFromCurve(FloatCurve=DataInterface() {propertyValues {Curve {keys [<at 0 1>,<at 0.15 0.5>,<at 0.4 0.3>]}}}))
     Float=&RandomRangeFloat(Minimum=0.35, Maximum=0.12)
    )
    ScaleA=true
    ScaleRGB=true
   )
   &Drag(Drag=3, Use Linear Drag=true, Use Rotational Drag=false)
   &SolveForcesAndVelocity(Clamp Velocity=false, Limit Acceleration=false)
   &ScaleSpriteSize(Scale Factor=&Vector2DFromFloat(Value=&FloatFromCurve(
      FloatCurve=DataInterface() {propertyValues {Curve {keys [
          <at 0 0>
          {arriveTangent=1.5202, interpMode=RCIM_Cubic, leaveTangent=1.5202, tangentMode=RCTM_User, time=0.1, value=1.7}
          <at 1.0011 3.4809 RCIM_Cubic>
         ]}}}
      Scale Curve=3
     )))
   /SetVariables(Particles.SubImageIndex=&FloatFromCurve(FloatCurve=DataInterface() {propertyValues {Curve {keys [{arriveTangent=29.8623, interpMode=RCIM_Cubic, leaveTangent=29.8623, tangentMode=RCTM_User, time=0, value=0}, <at 1.0011 22.1689 RCIM_Cubic>]}}})) {isSetParameters=true, module=SetVariables_#2}
   &DynamicMaterialParameters(
    Index 0 Param 1=&FloatFromCurve(FloatCurve=DataInterface() {propertyValues {Curve {keys [<at -0.0046 1.0113 RCIM_Cubic>,<at 1.0011 0.002>]}}})
    Index 0 Param 4=&FloatFromCurve(FloatCurve=DataInterface() {propertyValues {Curve {keys [<at 0 0 RCIM_Cubic>,<at 1 0.5>]}}})
    Module.Param0WriteEnabled=true
    Module.Param1WriteEnabled=false
    Module.Param2WriteEnabled=false
    Module.Param3WriteEnabled=true
   ) {module=DynamicMatParams}
   &CameraOffset(Camera Offset Amount=&RandomRangeFloat(Minimum=20, Maximum=30))
  ]
  rendererClasses [&NiagaraSpriteRendererProperties]
  renderers [{rendererClass=&NiagaraSpriteRendererProperties}]
 }
 SmokeTrail_01 {
  emitterUpdateScript [
   &EmitterScalability(
    Module.bApplySpawnCountScale=false
    Module.bApplySpawnCountScaleByDistanceFraction=false
    Module.bUseMaxDistance=false
    Module.bUseMinDistance=false
   ) {enabled=false}
   &EmitterState(Life Cycle Mode=::NewEnumerator1, Loop Behavior=::NewEnumerator1, Loop Duration=1, UseLoopDelay=false)
   &EmitterLifeCycle() {enabled=false}
   &SpawnPerUnit(Module.Use Max Movement Threshold=true, Module.Use Movement Tolerance=false, Spawn Per Unit=4, Use Spawn Probability=false) {enabled=false}
   &SpawnRate(SpawnRate=20, Use Spawn Probability=false) {enabled=false}
   &SpawnPerFrame(Spawn Count=2, Use Spawn Probability=true)
  ]
  particleSpawnScript [
   &InitializeRibbon(
    Color=<rgba 0 0.5 1 0.2>
    Lifetime=&RandomRangeFloat(Minimum=0.5, Maximum=0.5)
    Module.Write Color=true
    Module.Write Lifetime=true
    Module.Write Mass=true
    Module.Write Material Random=true
    Module.Write Position=true
    Module.Write RibbonTwist=true
    Module.Write RibbonWidth=true
    Position:Vector3f=&Add_Vector(A=/Niagara/DynamicInputs/Helpers/SimulationPosition(), B=<0 0 0>)
    Ribbon Twist=1
    Ribbon Width=40
   )
   &AccelerationForce(Acceleration=<0 0 50>) {module=AccelerationForce001}
   &AddVelocity(Scale Added Velocity=<2 2 1>, Velocity=&RandomRangeVector(Minimum=<2 2 20>, Maximum=<-2 -2 20>))
   &ApplyInitialForces()
   &FN_SetSpawnGroup_v2(Number of Spawn Groups=7)
   &SkeletalMeshLocation(
    Bone / Socket Index=&SpawnGroup()
    Bone Sampling Mode=::NewEnumerator4
    Inherit Velocity=false
    MaskSpawnGroup=false
    Skeletal Mesh=$User.SkeletalMesh
   )
  ]
  particleUpdateScript [
   &ParticleState()
   &AccelerationForce(Acceleration=<0 0 20>)
   &SolveForcesAndVelocity(Clamp Velocity=false, Limit Acceleration=false)
   &Color(Color=$Particles.Initial.Color)
   &DynamicMaterialParameters(
    Index 0 Param 1=&FloatFromCurve(FloatCurve=DataInterface() {propertyValues {Curve {keys [<at -0.0046 1.0113>,<at 0.5201 0.2342>,<at 0.7 0.1>]}}})
    Index 0 Param 2=0.1
    Index 0 Param 3=-5
    Index 0 Param 4=&FloatFromCurve(FloatCurve=DataInterface() {propertyValues {Curve {keys [<at -0.0046 1.0113>,<at 1 0>]}}})
    Module.Param0WriteEnabled=true
    Module.Param1WriteEnabled=true
    Module.Param2WriteEnabled=true
    Module.Param3WriteEnabled=true
   )
   &ScaleColor(
    Scale RGB=&VectorFromFloat(Value=&FloatFromCurve(FloatCurve=DataInterface() {propertyValues {Curve {keys [
          {arriveTangent=-6.1113, interpMode=RCIM_Cubic, leaveTangent=-6.1113, tangentMode=RCTM_User, time=0, value=1}
          {arriveTangent=-1.5434, interpMode=RCIM_Cubic, leaveTangent=-1.5434, tangentMode=RCTM_User, time=0.1302, value=0.4371}
          {arriveTangent=-0.7141, interpMode=RCIM_Cubic, leaveTangent=-0.7141, tangentMode=RCTM_User, time=0.3459, value=0.1927}
          {arriveTangent=-0.2078, interpMode=RCIM_Cubic, leaveTangent=-0.2078, tangentMode=RCTM_User, time=1, value=0}
         ]}}}))
    ScaleA=true
    ScaleRGB=true
   )
   &CameraOffset(Camera Offset Amount=25)
   &ScaleRibbonWidth(Ribbon Width Scale=&FloatFromCurve(FloatCurve=DataInterface() {propertyValues {Curve {keys [{interpMode=RCIM_Cubic, leaveTangent=-1.2546, tangentMode=RCTM_Break, time=0, value=1.5}, <at 1 1>]}}})) {enabled=false}
  ]
  rendererClasses [&NiagaraRibbonRendererProperties]
  renderers [{rendererClass=&NiagaraRibbonRendererProperties}]
 }
 Spike {
  emitterUpdateScript [
   &EmitterState()
   &SpawnBurst_Instantaneous(Use Spawn Probability=false)
  ]
  particleSpawnScript [
   &InitializeParticle(
    Color=<rgba 0.125 0.4167 1 2>
    Lifetime=&UniformRangedFloat(Minimum=0.1, Maximum=0.13)
    Module.Write Color=true
    Module.Write Lifetime=true
    Module.Write Mass=true
    Module.Write Material Random=true
    Module.Write Position=false
    Module.Write Scale=false
    Module.Write SpriteRotation=false
    Module.Write SpriteSize=true
    Position=/Niagara/Functions/Localspace/SimulationPosition()
    Sprite Size=&UniformRangedVector2D(Minimum=<0.1 0.9>, Maximum=<0.27 0.4>)
   )
   &AddVelocityFromPoint(Module.Use Falloff Distance=false, Velocity Strength=60)
  ]
  particleUpdateScript [
   &ParticleState()
   &SolveForcesAndVelocity(Clamp Velocity=false, Limit Acceleration=false)
   &ScaleColor(
    Scale Alpha=&FloatFromCurve(FloatCurve=DataInterface() {propertyValues {Curve {keys [<at 0 2>,<at 1 0>]}}})
    Scale RGB=&VectorFromFloat(Value=&FloatFromCurve(FloatCurve=DataInterface() {propertyValues {Curve {keys [<at 0 40>,<at 0.5 1>]}}}))
    ScaleA=true
    ScaleRGB=true
   )
   &ScaleSpriteSize(Scale Factor=&Vector2DFromFloat(Value=&FloatFromCurve(
      FloatCurve=DataInterface() {propertyValues {Curve {keys [{arriveTangent=2.4508, interpMode=RCIM_Cubic, leaveTangent=2.4508, tangentMode=RCTM_User, time=0, value=0.4}, <at 1 1.5 RCIM_Cubic>]}}}
      Scale Curve=5
     )))
  ]
  rendererClasses [&NiagaraSpriteRendererProperties]
  renderers [{rendererClass=&NiagaraSpriteRendererProperties}]
 }
}
summary {
 systemName=NS_Emote_ManAndMonster_Bubble_02
 emitters [
  {emitterName=Pops01, rendererClasses [&NiagaraSpriteRendererProperties]}
  {emitterName=Bubbles, rendererClasses [&NiagaraSpriteRendererProperties]}
  {emitterName=Pops02, rendererClasses [&NiagaraSpriteRendererProperties]}
  {emitterName=Spike, rendererClasses [&NiagaraSpriteRendererProperties]}
  {emitterName=SmokeTrail_01, rendererClasses [&NiagaraRibbonRendererProperties]}
  {emitterName=SmokeHeavy, rendererClasses [&NiagaraSpriteRendererProperties]}
  {emitterName=SmokeHeavy_Big, rendererClasses [&NiagaraSpriteRendererProperties]}
  {emitterName=Initial_Smoke, rendererClasses [&NiagaraSpriteRendererProperties]}
 ]
 userVariables [User.SkeletalMesh {
   type=NiagaraDataInterfaceSkeletalMesh
   defaultValue {
    struct=NiagaraExt_VariableValue_DataInterface
    value {dataInterface=&self:NiagaraDataInterfaceSkeletalMesh_0}
   }
  }]
}
