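# Niagara drip effect for the ManAndMonster emote, producing a short burst of teal falling droplets.
# A GPU sprite emitter and a sparser mesh emitter share one setup: spawn rate tapers over loop age and particles start in a tiny sphere.
# Downward acceleration, slight random drift and drag move the short-lived drops, while curves stretch mesh scale along X and thin Y/Z.
# User SkeletalMesh and IsSprinting parameters are exposed but unused by the modules shown.
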NiagaraSystem /BRCosmetics/Athena/Effects/Particles/Emotes/NS_Emote_ManAndMonster/NS_Emote_ManAndMonster_Drop_02
dependencies {
 usedDataInterfaces [&NiagaraDataInterfaceCurve, &NiagaraDataInterfaceVectorCurve]
 usedDynamicInputs []
 usedModules [&SystemState]
}
emitters {
 Drops_GPUParticle {
  simTarget=GPUComputeSim
  emitterUpdateScript [
   &EmitterState()
   &SpawnRate(
    SpawnRate=&FloatFromCurve(
     FloatCurve=DataInterface() {propertyValues {Curve {keys [
         {arriveTangent=-2.059, interpMode=RCIM_Cubic, leaveTangent=-2.059, tangentMode=RCTM_User, time=-0.0046, value=1.0113}
         {arriveTangent=-0.4473, interpMode=RCIM_Cubic, leaveTangent=-0.4473, tangentMode=RCTM_User, time=1.0011, value=0.002}
        ]}}}
     CurveIndex=$Emitter.NormalizedLoopAge
     Scale Curve=&RandomRangeFloat(Minimum=80, Maximum=70)
    )
    Use Spawn Probability=false
   )
  ]
  particleSpawnScript [
   /Niagara/Modules/Spawn/Initialization/InitializeParticle(
    Color=<rgba 0.0667 0.5 0.5 1>
    Lifetime=0.25
    Mass=&UniformRangedFloat(Minimum=0.6, Maximum=1.2)
    Module.Write Color=true
    Module.Write Lifetime=true
    Module.Write Mass=true
    Module.Write Material Random=true
    Module.Write Position=false
    Module.Write Scale=true
    Module.Write SpriteRotation=true
    Module.Write SpriteSize=true
    Position=&SimulationPosition()
    Sprite Rotation=180
    Sprite Size=&UniformRangedVector2D(Minimum=<0.5 8>, Maximum=<0.4 2>)
   )
   /Niagara/Modules/Spawn/Location/SphereLocation(MaskSpawnGroup=false, Module.Override Randomness=false, Sphere Radius=0.5, Surface Only=false)
  ]
  particleUpdateScript [
   &ParticleState()
   &AccelerationForce(Acceleration=&RandomRangeVector(Minimum=<0 0 -300>, Maximum=<1 1 -400>))
   &AddVelocity(Coordinate Space=::World, Velocity=&RandomRangeVector(Minimum=<0.2 0.2 0>, Maximum=<-0.2 -0.2 0>)) {module=AddVelocity001}
   &Drag(Drag=3, Use Linear Drag=true, Use Rotational Drag=false)
   &SolveForcesAndVelocity(Clamp Velocity=false, Limit Acceleration=false)
   &ScaleMeshSize(Scale Factor=&VectorFromCurve(VectorCurve=DataInterface() {propertyValues {
       XCurve {keys [
         {arriveTangent=2.7287, interpMode=RCIM_Cubic, leaveTangent=2.7287, tangentMode=RCTM_User, time=0, value=0}
         {arriveTangent=1, interpMode=RCIM_Cubic, leaveTangent=1, time=0.7, value=2}
         {arriveTangent=-2.9515, interpMode=RCIM_Cubic, leaveTangent=-2.9515, tangentMode=RCTM_User, time=1, value=1}
        ]}
       YCurve {keys [
         {interpMode=RCIM_Cubic, tangentMode=RCTM_Break, time=0, value=1}
         {arriveTangent=-1.1086, interpMode=RCIM_Cubic, tangentMode=RCTM_Break, time=1, value=0.5}
        ]}
       ZCurve {keys [
         {interpMode=RCIM_Cubic, tangentMode=RCTM_Break, time=0, value=1}
         {arriveTangent=-1.0363, interpMode=RCIM_Cubic, leaveTangent=-1.0363, tangentMode=RCTM_User, time=1, value=0.5}
        ]}
      }}))
  ]
  rendererClasses [&NiagaraSpriteRendererProperties]
  renderers [{rendererClass=&NiagaraSpriteRendererProperties}]
 }
 Drops_Mesh {
  emitterUpdateScript [
   &EmitterState()
   &SpawnRate(
    SpawnRate=&FloatFromCurve(
     FloatCurve=DataInterface() {propertyValues {Curve {keys [
         {arriveTangent=-2.059, interpMode=RCIM_Cubic, leaveTangent=-2.059, tangentMode=RCTM_User, time=-0.0046, value=1.0113}
         {arriveTangent=-0.4473, interpMode=RCIM_Cubic, leaveTangent=-0.4473, tangentMode=RCTM_User, time=1.0011, value=0.002}
        ]}}}
     CurveIndex=$Emitter.NormalizedLoopAge
     Scale Curve=&RandomRangeFloat(Minimum=25, Maximum=30)
    )
    Use Spawn Probability=false
   )
  ]
  particleSpawnScript [
   &InitializeParticle(
    Color=<rgba 0.0667 0.3778 0.5 1>
    Lifetime=0.25
    Mass=&UniformRangedFloat(Minimum=0.6, Maximum=1.2)
    Mesh Scale=&RandomRangeVector(Minimum=<0.6 0.07 0.07>, Maximum=<1 0.15 0.15>)
    Module.Write Color=true
    Module.Write Lifetime=true
    Module.Write Mass=true
    Module.Write Material Random=true
    Module.Write Position=false
    Module.Write Scale=true
    Module.Write SpriteRotation=true
    Module.Write SpriteSize=true
    Position=&SimulationPosition()
    Sprite Rotation=&NormalizedAngleToDegrees(NormalizedAngle=&UniformRangedFloat(Minimum=0, Maximum=1))
    Sprite Size=<4 20>
   )
   &SphereLocation(MaskSpawnGroup=false, Module.Override Randomness=false, Sphere Radius=0.5, Surface Only=false)
  ]
  particleUpdateScript [
   &ParticleState()
   &AccelerationForce(Acceleration=&RandomRangeVector(Minimum=<0 0 -300>, Maximum=<1 1 -400>))
   &AddVelocity(Coordinate Space=::World, Velocity=&RandomRangeVector(Minimum=<0.1 0.1 0>, Maximum=<-0.1 -0.1 0>)) {module=AddVelocity001}
   &Drag(Drag=3, Use Linear Drag=true, Use Rotational Drag=false)
   &SolveForcesAndVelocity(Clamp Velocity=false, Limit Acceleration=false)
   &ScaleMeshSize(Scale Factor=&VectorFromCurve(VectorCurve=DataInterface() {propertyValues {
       XCurve {keys [<at 0 0>,<at 1 2>]}
       YCurve {keys [
         {interpMode=RCIM_Cubic, tangentMode=RCTM_Break, time=0, value=1}
         {arriveTangent=-1.1086, interpMode=RCIM_Cubic, tangentMode=RCTM_Break, time=1, value=0.5}
        ]}
       ZCurve {keys [
         {interpMode=RCIM_Cubic, tangentMode=RCTM_Break, time=0, value=1}
         {arriveTangent=-1.0363, interpMode=RCIM_Cubic, leaveTangent=-1.0363, tangentMode=RCTM_User, time=1, value=0.5}
        ]}
      }}))
  ]
  rendererClasses [&NiagaraMeshRendererProperties]
  renderers [{rendererClass=&NiagaraMeshRendererProperties, rendererIndex=0}]
 }
}
summary {
 systemName=NS_Emote_ManAndMonster_Drop_02
 emitters [
  {emitterName=Drops_Mesh, rendererClasses [&NiagaraMeshRendererProperties]}
  {emitterName=Drops_GPUParticle, simTarget=GPUComputeSim, rendererClasses [&NiagaraSpriteRendererProperties]}
 ]
 userVariables [
  User.IsSprinting {type=NiagaraBool}
  User.SkeletalMesh {
   type=NiagaraDataInterfaceSkeletalMesh
   defaultValue {
    struct=NiagaraExt_VariableValue_DataInterface
    value {dataInterface=&self:NiagaraDataInterfaceSkeletalMesh_0}
   }
  }
 ]
}
