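# Niagara effect for the Man and Monster emote that bursts blue glowing sprites and smoke over the character's body, with a brief light flash.
# ParticleDecal and SmokeHeavy sample bone or socket positions from the user skeletal mesh via spawn groups, while SmokeHeavyHat targets the head.
# Initial_Dots adds noisy drifting sparks, and Smoke adds a darkening flipbook puff around the body.
# Curves drive dynamic material parameters on the decal-like sprites and heavy smoke.
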NiagaraSystem /BRCosmetics/Athena/Effects/Particles/Emotes/NS_Emote_ManAndMonster/NS_Emote_ManAndMonster_ParticleDecal
dependencies {
 usedDataInterfaces [&NiagaraDataInterfaceColorCurve, &NiagaraDataInterfaceCurve]
 usedDynamicInputs []
 usedModules [&SystemState]
}
emitters {
 Initial_Dots {
  emitterUpdateScript [
   &EmitterState(Life Cycle Mode=::NewEnumerator1, Loop Behavior=::NewEnumerator1, Loop Duration=0.5, UseLoopDelay=false)
   &SpawnRate(SpawnRate=30, Use Spawn Probability=false)
  ]
  particleSpawnScript [
   /Niagara/Modules/Spawn/Initialization/InitializeParticle(
    Lifetime=&UniformRangedFloat(Minimum=0.4, Maximum=1)
    Module.Write Color=true
    Module.Write Lifetime=true
    Module.Write Mass=true
    Module.Write Material Random=true
    Module.Write Position=false
    Module.Write Scale=false
    Module.Write SpriteRotation=false
    Module.Write SpriteSize=true
    Position=&SimulationPosition()
    Sprite Size=&Vector2DFromFloat(Value=&Multiply_Float(A=&UniformRangedFloat(Minimum=1, Maximum=3), B=1))
   )
   /Niagara/Modules/Spawn/Location/SphereLocation(
    MaskSpawnGroup=false
    Module.Override Randomness=false
    Non Uniform Scale=<0.6 0.6 1.25>
    Offset=<0 0 50>
    Sphere Radius=50
    Surface Only=false
   )
   &AddVelocity(
    Coordinate Space=::World
    Velocity=&MakeVector(X=0, Y=0, Z=&UniformRangedFloat(Minimum=50, Maximum=0))
   )
   &AddVelocityFromPoint(Module.Use Falloff Distance=false, Velocity Strength=50)
  ]
  particleUpdateScript [
   &ParticleState()
   &Color(
    Color=&ColorFromCurve(ColorCurve=DataInterface() {propertyValues {
       AlphaCurve {keys [<at 0 1 RCIM_Cubic>,<at 1 1 RCIM_Cubic>]}
       BlueCurve {keys [<at 0 1 RCIM_Cubic>,<at 0.7701 0.1302 RCIM_Cubic>]}
       GreenCurve {keys [<at 0 0.475 RCIM_Cubic>,<at 0.7701 0.0956 RCIM_Cubic>]}
       RedCurve {keys [<at 0 0.1 RCIM_Cubic>,<at 0.7701 0.0264 RCIM_Cubic>]}
      }})
    Scale Alpha=&FloatFromCurve(FloatCurve=DataInterface() {propertyValues {Curve {keys [{interpMode=RCIM_Cubic, leaveTangent=0.0979, tangentMode=RCTM_Break, time=0, value=1}, <at 1 0>]}}})
   )
   &ScaleSpriteSize(Scale Factor=&Vector2DFromFloat(Value=&FloatFromCurve(
      FloatCurve=DataInterface() {propertyValues {Curve {keys [
          {interpMode=RCIM_Cubic, leaveTangent=12.9788, tangentMode=RCTM_Break, time=0, value=0}
          {arriveTangent=-2.061, interpMode=RCIM_Cubic, leaveTangent=-4.7196, tangentMode=RCTM_Break, time=0.3046, value=1.1}
          {interpMode=RCIM_Cubic, tangentMode=RCTM_Break, time=1, value=0}
         ]}}}
      Scale Curve=&FloatFromCurve(FloatCurve=DataInterface() {propertyValues {Curve {keys [{interpMode=RCIM_Cubic, tangentMode=RCTM_Break, time=0, value=0.1}, <at 0.4 1>]}}})
     )))
   &ScaleVelocity(Velocity Scale=&MakeVector(
     X=1
     Y=1
     Z=&FloatFromCurve(FloatCurve=DataInterface() {propertyValues {Curve {keys [<at 0 1 RCIM_Cubic>, {arriveTangent=3.6879, interpMode=RCIM_Cubic, leaveTangent=3.6879, tangentMode=RCTM_User, time=1, value=2.5}]}}})
    ))
   &Drag(Drag=5, Use Linear Drag=true, Use Rotational Drag=false)
   &VectorNoiseForce(Force Amount=500, Module.Use Random Noise Time Interval=false)
   &SolveForcesAndVelocity(Clamp Velocity=false, Limit Acceleration=false)
  ]
  rendererClasses [&NiagaraSpriteRendererProperties]
  renderers [{rendererClass=&NiagaraSpriteRendererProperties}]
 }
 Light {
  emitterUpdateScript [
   &EmitterState(Life Cycle Mode=::NewEnumerator1, Loop Behavior=::NewEnumerator1, Loop Duration=1, UseLoopDelay=false)
   &SpawnBurst_Instantaneous(Spawn Count=1, Use Spawn Probability=false)
  ]
  particleSpawnScript [
   &InitializeParticle(
    Color=<rgba 0.1 0.55 1 1>
    Lifetime=0.25
    Module.Write Color=true
    Module.Write Lifetime=true
    Module.Write Mass=true
    Module.Write Material Random=true
    Module.Write Position=false
    Module.Write Scale=false
    Module.Write SpriteRotation=false
    Module.Write SpriteSize=true
    Position=&SimulationPosition()
    Sprite Size=<200 200>
   )
   &SystemLocation(MaskSpawnGroup=false, Offset=<0 0 80>)
  ]
  particleUpdateScript [
   &ParticleState()
   &ScaleColor(
    Scale Alpha=&FloatFromCurve(FloatCurve=DataInterface() {propertyValues {Curve {keys [
         {arriveTangent=-0.551, interpMode=RCIM_Cubic, leaveTangent=-0.551, tangentMode=RCTM_User, time=0, value=1}
         {arriveTangent=-2.3347, interpMode=RCIM_Cubic, leaveTangent=-2.3347, tangentMode=RCTM_User, time=1.0011, value=0.002}
        ]}}})
    ScaleA=true
    ScaleRGB=true
   )
   &ScaleSpriteSize(Scale Factor=&Vector2DFromFloat(Value=&FloatFromCurve(FloatCurve=DataInterface() {propertyValues {Curve {keys [<at 0.0239 0>,<at 0.1 1>,<at 1.0011 0.002>]}}}, Scale Curve=2)))
   &FN_Light_v2(
    Light Exponent=2
    Light Radius=100
    LightEnabled=true
    Scale Color:NiagaraFloat=&FloatFromCurve(FloatCurve=DataInterface() {propertyValues {Curve {keys [<at 0 2>,<at 1 0>]}}})
   )
  ]
  rendererClasses [&NiagaraLightRendererProperties]
  renderers [{rendererClass=&NiagaraLightRendererProperties, rendererIndex=0}]
 }
 ParticleDecal {
  emitterUpdateScript [
   &EmitterState()
   &SpawnBurst_Instantaneous(Spawn Count=17, Use Spawn Probability=false)
  ]
  particleSpawnScript [
   &InitializeParticle(
    Color=&ColorFromCurve(ColorCurve=DataInterface() {propertyValues {
       AlphaCurve {keys [<at 0 1 RCIM_Cubic>]}
       BlueCurve {keys [<at 0.0406 25>, {arriveTangent=-27.464, interpMode=RCIM_Cubic, leaveTangent=-27.464, time=0.2063, value=6}, <at 0.9378 0.36 RCIM_Cubic>]}
       GreenCurve {keys [<at 0.0406 13.5>, {arriveTangent=-11.7034, interpMode=RCIM_Cubic, leaveTangent=-11.7034, time=0.2063, value=3.7}, <at 0.9378 3 RCIM_Cubic>]}
       RedCurve {keys [<at 0.0406 2>, {arriveTangent=0.1338, interpMode=RCIM_Cubic, leaveTangent=0.1338, time=0.2063, value=0.48}, <at 0.9378 2.12 RCIM_Cubic>]}
      }})
    Lifetime=0.55
    Module.Write Color=true
    Module.Write Lifetime=true
    Module.Write Mass=true
    Module.Write Material Random=true
    Module.Write Position=true
    Module.Write Scale=false
    Module.Write SpriteRotation=true
    Module.Write SpriteSize=true
    Position=&SimulationPosition()
    Sprite Size=&Vector2DFromFloat(Value=75)
   )
   &CameraOffset(Camera Offset Amount=25)
   &FN_SetSpawnGroup_v2(Number of Spawn Groups=10)
   &SkeletalMeshLocation(
    Bone / Socket Index=&SpawnGroup()
    Bone Sampling Mode=::NewEnumerator4
    Inherit Velocity=false
    MaskSpawnGroup=false
    Sampled Position Offset=&RandomRangeVector(Minimum=<7 7 5>, Maximum=<-7 -7 -5>)
    Skeletal Mesh=$User.SkeletalMesh
   )
  ]
  particleUpdateScript [
   &ParticleState()
   &ScaleColor(
    Scale Alpha=&FloatFromCurve(FloatCurve=DataInterface() {propertyValues {Curve {keys [
         <at 0 0 RCIM_Cubic>
         {arriveTangent=-0.5034, interpMode=RCIM_Cubic, leaveTangent=-0.5034, tangentMode=RCTM_User, time=0.3, value=1}
         <at 1 0>
        ]}}})
    ScaleA=true
    ScaleRGB=true
   )
   &DynamicMaterialParameters(
    Index 0 Param 1=&FloatFromCurve(FloatCurve=DataInterface() {propertyValues {Curve {keys [
         {interpMode=RCIM_Cubic, leaveTangent=8.739, tangentMode=RCTM_Break, time=0, value=-0.5}
         {arriveTangent=19.7324, interpMode=RCIM_Cubic, leaveTangent=19.7324, time=0.0514, value=0.8887}
         {arriveTangent=8.2492, interpMode=RCIM_Cubic, leaveTangent=8.2492, tangentMode=RCTM_User, time=0.1496, value=2.4517}
         {arriveTangent=4.695, interpMode=RCIM_Cubic, tangentMode=RCTM_Break, time=1, value=5}
        ]}}})
    Index 0 Param 2=1
    Index 0 Param 3=&FloatFromCurve(FloatCurve=DataInterface() {propertyValues {Curve {keys [
         {interpMode=RCIM_Cubic, leaveTangent=2.9398, tangentMode=RCTM_Break, time=0, value=-0.25}
         {arriveTangent=0.3723, interpMode=RCIM_Cubic, tangentMode=RCTM_Break, time=1, value=1.4}
        ]}}})
    Index 0 Param 4=1
    Module.Param0WriteEnabled=true
    Module.Param1WriteEnabled=true
    Module.Param2WriteEnabled=true
    Module.Param3WriteEnabled=true
   )
   &SolveForcesAndVelocity(Clamp Velocity=false, Limit Acceleration=false)
  ]
  rendererClasses [&NiagaraSpriteRendererProperties]
  renderers [{rendererClass=&NiagaraSpriteRendererProperties}]
 }
 Smoke {
  emitterUpdateScript [
   &EmitterState()
   &SpawnRate(SpawnRate=50, Use Spawn Probability=false) {enabled=false}
   &SpawnBurst_Instantaneous(Spawn Count=8, Spawn Time=0.1, Use Spawn Probability=false)
  ]
  particleSpawnScript [
   &InitializeParticle(
    Lifetime=&UniformRangedFloat(Minimum=1.3, Maximum=1.7)
    Module.Write Color=true
    Module.Write Lifetime=true
    Module.Write Mass=true
    Module.Write Material Random=true
    Module.Write Position=false
    Module.Write Scale=false
    Module.Write SpriteRotation=true
    Module.Write SpriteSize=true
    Position=&SimulationPosition()
    Sprite Rotation=&UniformRangedFloat(Minimum=0, Maximum=180)
    Sprite Size=&UniformRangedVector2D(Minimum=<200 200>, Maximum=<150 150>)
   )
   &AddVelocity(Velocity=<0 0 15>)
   &CylinderLocation(
    Cylinder Height=55
    Cylinder Radius=30
    Hemisphere X=true
    Lathe Profile=DataInterface() {propertyValues {Curve {keys [<at 0 0>,<at 0.9 0>,<at 1.0013 1>]}}}
    MaskSpawnGroup=false
    Module.Enable Lathe Profile=false
    Module.Override Randomness=false
    Module.Surface Only=true
    Offset=<-20 0 70>
    Surface Only Band Thickness=0
    Use Endcaps In Surface Only Mode=false
   )
  ]
  particleUpdateScript [
   &ParticleState()
   /Niagara/Modules/Update/Forces/CurlNoiseForce(Noise Frequency=0.2, Noise Strength=25, Pan Noise=false)
   &Drag(Drag=3, Use Linear Drag=true, Use Rotational Drag=false)
   &SolveForcesAndVelocity(Clamp Velocity=false, Limit Acceleration=false)
   /Niagara/Modules/Update/SubUV/SubUVAnimation(
    Number Of Frames=64
    SubUV Lookup Index=&FloatFromCurve(FloatCurve=DataInterface() {propertyValues {Curve {keys [
         {arriveTangent=0.7985, interpMode=RCIM_Cubic, leaveTangent=0.7985, tangentMode=RCTM_User, time=0, value=0.6}
         {arriveTangent=0.0289, interpMode=RCIM_Cubic, leaveTangent=0.0289, tangentMode=RCTM_User, time=1.0011, value=0.985}
        ]}}})
   )
   &ScaleSpriteSize(Scale Factor=&Vector2DFromFloat(Value=&FloatFromCurve(
      FloatCurve=DataInterface() {propertyValues {Curve {keys [
          <at -0.0046 0.25 RCIM_Cubic>
          {arriveTangent=-0.1882, interpMode=RCIM_Cubic, leaveTangent=-0.1882, tangentMode=RCTM_User, time=1.0099, value=0.8547}
         ]}}}
      Scale Curve=0.85
     )))
   &Color(Color=&ColorFromCurve(ColorCurve=DataInterface() {propertyValues {
       AlphaCurve {keys [<at -0.0382 0>,<at 0.004 1 RCIM_Cubic>,<at 0.9977 0 RCIM_Cubic>]}
       BlueCurve {keys [<at 0.0273 0.6198>,<at 0.1487 0.026>]}
       GreenCurve {keys [<at 0.0273 0.5334>,<at 0.1487 0.0158>]}
       RedCurve {keys [<at 0.0273 0.1015>,<at 0.1487 0.0056>]}
      }}))
  ]
  rendererClasses [&NiagaraSpriteRendererProperties]
  renderers [{rendererClass=&NiagaraSpriteRendererProperties}]
 }
 SmokeHeavy {
  emitterUpdateScript [
   &EmitterState(Life Cycle Mode=::NewEnumerator1, Loop Behavior=::NewEnumerator1, Loop Duration=0.8, UseLoopDelay=true)
   &SpawnRate(
    SpawnRate=&FloatFromCurve(
     FloatCurve=DataInterface() {propertyValues {Curve {keys [<at 0 0>,<at 0.2 0.3>,<at 0.4 1>,<at 5.5013 1>]}}}
     CurveIndex=$Emitter.NormalizedLoopAge
     Scale Curve=300
    )
    Use Spawn Probability=false
   ) {enabled=false}
   &SpawnBurst_Instantaneous(Spawn Count=40, Use Spawn Probability=false)
  ]
  particleSpawnScript [
   &InitializeParticle(
    Color=&ColorFromCurve(ColorCurve=DataInterface() {propertyValues {
       AlphaCurve {keys [<at -0.0445 0.15>]}
       BlueCurve {keys [<at 0 3 RCIM_Cubic>,<at 0.0727 0.7>,<at 0.6644 0.3>]}
       GreenCurve {keys [<at 0 2.1333 RCIM_Cubic>,<at 0.0727 0.2956>,<at 0.6644 0.015>]}
       RedCurve {keys [<at 0 0.4 RCIM_Cubic>,<at 0.0727 0.0933>,<at 0.6644 0.015>]}
      }})
    Lifetime=&UniformRangedFloat(Minimum=0.85, Maximum=0.5)
    Module.Write Color=true
    Module.Write Lifetime=true
    Module.Write Mass=true
    Module.Write Material Random=true
    Module.Write Position=false
    Module.Write Scale=false
    Module.Write SpriteRotation=true
    Module.Write SpriteSize=true
    Position=&SimulationPosition()
    Sprite Rotation=&UniformRangedFloat(Minimum=0, Maximum=360)
    Sprite Size=&Vector2DFromFloat(Value=&UniformRangedFloat(Minimum=3, Maximum=5.5))
   )
   &AccelerationForce(Acceleration=&UniformRangedVector(Minimum=<0 0 15>, Maximum=<0 0 60>))
   &AddVelocity(Scale Added Velocity=<0.8 0.8 1>, Velocity=&UniformRangedVector(Minimum=<-20 -20 5>, Maximum=<20 20 20>))
   &ApplyInitialForces()
   &FN_SetSpawnGroup_v2(Number of Spawn Groups=16)
   &SkeletalMeshLocation(
    Bone / Socket Index=&SpawnGroup()
    Bone Sampling Mode=::NewEnumerator4
    Inherit Velocity=false
    MaskSpawnGroup=false
    Sampled Position Offset=&RandomRangeVector(Minimum=<4 4 5>, Maximum=<-4 -4 -2>)
    Skeletal Mesh=$User.SkeletalMesh
   )
  ]
  particleUpdateScript [
   &ParticleState()
   &Drag(Drag=3, Use Linear Drag=true, Use Rotational Drag=false)
   &SolveForcesAndVelocity(Clamp Velocity=false, Limit Acceleration=false)
   &ScaleColor(
    Scale Alpha=&FloatFromCurve(
     FloatCurve=DataInterface() {propertyValues {Curve {keys [
         {arriveTangent=-3.1757, interpMode=RCIM_Cubic, leaveTangent=-3.1757, tangentMode=RCTM_User, time=0.5, value=2.5}
         {arriveTangent=-1.5121, interpMode=RCIM_Cubic, leaveTangent=-1.5121, tangentMode=RCTM_User, time=1.0011, value=0.002}
        ]}}}
     Scale Curve=0.85
    )
    Scale RGB=&Multiply_VectorByFloat(Vector=<0.3 0.3 0.3>, Float=&RandomRangeFloat(Minimum=0.6, Maximum=0.1))
    ScaleA=true
    ScaleRGB=true
   )
   &ScaleSpriteSize(Scale Factor=&Vector2DFromFloat(Value=&FloatFromCurve(
      FloatCurve=DataInterface() {propertyValues {Curve {keys [
          <at 0 0>
          {arriveTangent=1.5202, interpMode=RCIM_Cubic, leaveTangent=1.5202, tangentMode=RCTM_User, time=0.1, value=1.4}
          {arriveTangent=3.9459, interpMode=RCIM_Cubic, leaveTangent=3.9459, tangentMode=RCTM_User, time=1.0011, value=3.6}
         ]}}}
      Scale Curve=2.4
     )))
   /SetVariables(Particles.SubImageIndex=&FloatFromCurve(FloatCurve=DataInterface() {propertyValues {Curve {keys [{arriveTangent=29.8623, interpMode=RCIM_Cubic, leaveTangent=29.8623, tangentMode=RCTM_User, time=0, value=0}, <at 1.0011 22.1689 RCIM_Cubic>]}}})) {isSetParameters=true, module=SetVariables_#1}
   &DynamicMaterialParameters(
    Index 0 Param 1=&FloatFromCurve(FloatCurve=DataInterface() {propertyValues {Curve {keys [<at -0.0046 1.0113 RCIM_Cubic>,<at 1.0011 0.002>]}}})
    Index 0 Param 4=&FloatFromCurve(FloatCurve=DataInterface() {propertyValues {Curve {keys [<at 0 0 RCIM_Cubic>,<at 1 0.5>]}}})
    Module.Param0WriteEnabled=true
    Module.Param1WriteEnabled=false
    Module.Param2WriteEnabled=false
    Module.Param3WriteEnabled=true
   ) {module=DynamicMatParams}
   &CameraOffset(Camera Offset Amount=15)
  ]
  rendererClasses [&NiagaraSpriteRendererProperties]
  renderers [{rendererClass=&NiagaraSpriteRendererProperties}]
 }
 SmokeHeavyHat {
  emitterUpdateScript [
   &EmitterState(Life Cycle Mode=::NewEnumerator1, Loop Behavior=::NewEnumerator1, Loop Duration=0.4, UseLoopDelay=false)
   &SpawnRate(Use Spawn Probability=false)
   &SpawnBurst_Instantaneous(Spawn Count=3, Use Spawn Probability=false)
  ]
  particleSpawnScript [
   &InitializeParticle(
    Color=&ColorFromCurve(ColorCurve=DataInterface() {propertyValues {
       AlphaCurve {keys [<at -0.0445 0.15>]}
       BlueCurve {keys [<at 0 3 RCIM_Cubic>,<at 0.0727 0.7>,<at 0.6644 0.3>]}
       GreenCurve {keys [<at 0 2.1333 RCIM_Cubic>,<at 0.0727 0.2956>,<at 0.6644 0.015>]}
       RedCurve {keys [<at 0 0.4 RCIM_Cubic>,<at 0.0727 0.0933>,<at 0.6644 0.015>]}
      }})
    Lifetime=&UniformRangedFloat(Minimum=1, Maximum=1.2)
    Module.Write Color=true
    Module.Write Lifetime=true
    Module.Write Mass=true
    Module.Write Material Random=true
    Module.Write Position=false
    Module.Write Scale=false
    Module.Write SpriteRotation=true
    Module.Write SpriteSize=true
    Position=&SimulationPosition()
    Sprite Rotation=&UniformRangedFloat(Minimum=0, Maximum=360)
    Sprite Size=&Vector2DFromFloat(Value=&UniformRangedFloat(Minimum=8, Maximum=12))
   )
   &AccelerationForce(Acceleration=&UniformRangedVector(Minimum=<0 0 60>, Maximum=<0 0 75>))
   &AddVelocity(Scale Added Velocity=<0.8 0.8 1>, Velocity=&UniformRangedVector(Minimum=<0 0 60>, Maximum=<0 0 70>))
   &ApplyInitialForces()
   &FN_SetSpawnGroup_v2(Number of Spawn Groups=16)
   &SkeletalMeshLocation(
    Bone / Socket Index=/Niagara/DynamicInputs/Meshes/SpecificSocket(
     Mesh:NiagaraDataInterfaceSkeletalMesh=DataInterface() {propertyValues {
       DefaultMesh=null
       ExcludeBoneName=root
       OptionalComponentClass=null
       PreviewMesh=/Game/Characters/Player/Male/Medium/Heads/M_MED_Hyde_Head_01/Meshes/Male_Medium_Hyde_Head
       SkinningMode=SkinOnTheFly
       SoftSourceActor=null
       SourceMode=AttachParent
       UvSetIndex=0
       WholeMeshLOD=-1
       bExcludeBone=false
       bReadDeformedGeometry=true
       bRequireCurrentFrameData=true
       ComponentTags []
       FilteredBones [head]
       FilteredSockets []
       MeshUserParameter {parameter {name=None, typeDefHandle {registeredTypeIndex=1}, varData []}}
       SamplingRegions []
      }}
     SpecificSocketIndex:NiagaraInt32=0
    )
    Bone Sampling Mode=::NewEnumerator4
    Inherit Velocity=true
    "Inherit Velocity (Scale)"=0.1
    MaskSpawnGroup=false
    Sampled Position Offset=&RandomRangeVector(Minimum=<0 0 4>, Maximum=<0 0 8>)
    Sampled Position Offset Coordinate Space:ENiagaraMeshTransforms=::NewEnumerator2
    Skeletal Mesh=$User.SkeletalMesh
    VelocitySpeedLimit=false
   )
  ]
  particleUpdateScript [
   &ParticleState()
   &Drag(Drag=6, Use Linear Drag=true, Use Rotational Drag=false)
   &SolveForcesAndVelocity(Clamp Velocity=false, Limit Acceleration=false)
   &ScaleColor(
    Scale Alpha=&FloatFromCurve(
     FloatCurve=DataInterface() {propertyValues {Curve {keys [
         {arriveTangent=-0.2553, interpMode=RCIM_Cubic, leaveTangent=-0.2553, tangentMode=RCTM_User, time=0.5, value=2.5}
         {arriveTangent=-1.5121, interpMode=RCIM_Cubic, leaveTangent=-1.5121, tangentMode=RCTM_User, time=1.0011, value=0.002}
        ]}}}
     Scale Curve=1.25
    )
    Scale RGB=&Multiply_VectorByFloat(Vector=<0.3 0.3 0.3>, Float=&RandomRangeFloat(Minimum=0.3, Maximum=0.35))
    ScaleA=true
    ScaleRGB=true
   )
   &ScaleSpriteSize(Scale Factor=&Vector2DFromFloat(Value=&FloatFromCurve(
      FloatCurve=DataInterface() {propertyValues {
        bOptimizeLUT=false
        Curve {keys [
          <at 0 0>
          {arriveTangent=1.5202, interpMode=RCIM_Cubic, leaveTangent=1.5202, tangentMode=RCTM_User, time=0.063, value=1.8594}
          {arriveTangent=3.9459, interpMode=RCIM_Cubic, leaveTangent=3.9459, tangentMode=RCTM_User, time=0.8006, value=3.5}
         ]}
       }}
      Scale Curve=2.4
     )))
   /SetVariables(Particles.SubImageIndex=&FloatFromCurve(FloatCurve=DataInterface() {propertyValues {Curve {keys [{arriveTangent=29.8623, interpMode=RCIM_Cubic, leaveTangent=29.8623, tangentMode=RCTM_User, time=0, value=0}, <at 1.0011 22.1689 RCIM_Cubic>]}}})) {isSetParameters=true, module=SetVariables_#1}
   &DynamicMaterialParameters(
    Index 0 Param 1=&FloatFromCurve(FloatCurve=DataInterface() {propertyValues {Curve {keys [<at -0.0046 1.0113 RCIM_Cubic>,<at 1.0011 0.002>]}}})
    Index 0 Param 4=&FloatFromCurve(FloatCurve=DataInterface() {propertyValues {Curve {keys [<at 0 0 RCIM_Cubic>,<at 1 0.5>]}}})
    Module.Param0WriteEnabled=true
    Module.Param1WriteEnabled=false
    Module.Param2WriteEnabled=false
    Module.Param3WriteEnabled=true
   ) {module=DynamicMatParams}
   &CameraOffset(Camera Offset Amount=25)
  ]
  rendererClasses [&NiagaraSpriteRendererProperties]
  renderers [{rendererClass=&NiagaraSpriteRendererProperties}]
 }
}
summary {
 systemName=NS_Emote_ManAndMonster_ParticleDecal
 emitters [
  {emitterName=ParticleDecal, rendererClasses [&NiagaraSpriteRendererProperties]}
  {emitterName=Smoke, rendererClasses [&NiagaraSpriteRendererProperties]}
  {emitterName=Light, rendererClasses [&NiagaraLightRendererProperties]}
  {emitterName=Initial_Dots, rendererClasses [&NiagaraSpriteRendererProperties]}
  {emitterName=SmokeHeavy, rendererClasses [&NiagaraSpriteRendererProperties]}
  {emitterName=SmokeHeavyHat, rendererClasses [&NiagaraSpriteRendererProperties]}
 ]
 userVariables [User.SkeletalMesh {
   type=NiagaraDataInterfaceSkeletalMesh
   defaultValue {
    struct=NiagaraExt_VariableValue_DataInterface
    value {dataInterface=&self:NiagaraDataInterfaceSkeletalMesh_0}
   }
  }]
}
