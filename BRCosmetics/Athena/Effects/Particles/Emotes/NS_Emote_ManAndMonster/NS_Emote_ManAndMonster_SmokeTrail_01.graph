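# Niagara system for the ManAndMonster emote that draws translucent blue smoke trails with two ribbon emitters.
# SmokeTrail001 spawns by distance moved with longer-lived particles, while SmokeTrail_01 spawns at a steady rate with shorter-lived ones.
# Both add random sideways drift and upward acceleration, animate material parameters over life and darken color by curve.
# Ribbon-width curves are disabled, and User.IsSprinting is declared without visible use.
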NiagaraSystem /BRCosmetics/Athena/Effects/Particles/Emotes/NS_Emote_ManAndMonster/NS_Emote_ManAndMonster_SmokeTrail_01
dependencies {usedDataInterfaces [&NiagaraDataInterfaceCurve], usedDynamicInputs [], usedModules [&SystemState]}
emitters {
 SmokeTrail001 {
  emitterUpdateScript [
   &EmitterScalability(
    Module.bApplySpawnCountScale=false
    Module.bApplySpawnCountScaleByDistanceFraction=false
    Module.bUseMaxDistance=false
    Module.bUseMinDistance=false
   ) {enabled=false}
   &EmitterState(Life Cycle Mode=::NewEnumerator1, Loop Behavior=::NewEnumerator1, Loop Duration=1.1, UseLoopDelay=false)
   &EmitterLifeCycle() {enabled=false}
   &SpawnPerUnit(Module.Use Max Movement Threshold=true, Module.Use Movement Tolerance=false, Spawn Group=1, Spawn Per Unit=2, Use Spawn Probability=false)
  ]
  particleSpawnScript [
   &InitializeRibbon(
    Color=<rgba 0 0.6667 1 0.4>
    Lifetime=1.5
    Module.Write Color=true
    Module.Write Lifetime=true
    Module.Write Mass=true
    Module.Write Material Random=true
    Module.Write Position=true
    Module.Write RibbonTwist=true
    Module.Write RibbonWidth=true
    Position:Vector3f=&Add_Vector(A=/Niagara/DynamicInputs/Helpers/SimulationPosition(), B=<0 0 0>)
    Ribbon Twist=1
    Ribbon Width=20
   )
   &AccelerationForce(Acceleration=<0 0 10>) {module=AccelerationForce001}
   &AddVelocity(Scale Added Velocity=<3 3 1>, Velocity=&RandomRangeVector(Minimum=<2 2 0>, Maximum=<-2 -2 0>))
   &ApplyInitialForces()
  ]
  particleUpdateScript [
   &ParticleState()
   &AccelerationForce(Acceleration=&RandomRangeVector(Minimum=<0 0 50>, Maximum=<0 0 100>))
   &SolveForcesAndVelocity(Clamp Velocity=false, Limit Acceleration=false)
   &Color(Color=$Particles.Initial.Color)
   /SetVariables(Particles.RibbonWidth=&FloatFromCurve(FloatCurve=DataInterface() {propertyValues {Curve {keys [<at 0 1>,<at 1 0>]}}}, Scale Curve=6)) {enabled=false, isSetParameters=true, module=SetVariables_#1}
   &DynamicMaterialParameters(
    Index 0 Param 1=&FloatFromCurve(FloatCurve=DataInterface() {propertyValues {Curve {keys [<at -0.0046 1.0113>,<at 0.5201 0.2342>,<at 0.7 0.1>]}}})
    Index 0 Param 2=0.1
    Index 0 Param 3=-3
    Index 0 Param 4=&FloatFromCurve(FloatCurve=DataInterface() {propertyValues {Curve {keys [<at -0.0046 1.0113>,<at 1 0>]}}})
    Module.Param0WriteEnabled=true
    Module.Param1WriteEnabled=true
    Module.Param2WriteEnabled=true
    Module.Param3WriteEnabled=true
   )
   &ScaleColor(
    Scale Alpha=1.5
    Scale RGB=&VectorFromFloat(Value=&FloatFromCurve(FloatCurve=DataInterface() {propertyValues {Curve {keys [
          {arriveTangent=-0.0519, interpMode=RCIM_Cubic, leaveTangent=-0.0519, tangentMode=RCTM_User, time=0, value=1}
          {arriveTangent=-0.9379, interpMode=RCIM_Cubic, leaveTangent=-0.9379, tangentMode=RCTM_User, time=0.2578, value=0.2076}
          {arriveTangent=-0.2078, interpMode=RCIM_Cubic, leaveTangent=-0.2078, tangentMode=RCTM_User, time=1, value=0}
         ]}}}))
    ScaleA=true
    ScaleRGB=true
   )
  ]
  rendererClasses [&NiagaraRibbonRendererProperties]
  renderers [{rendererClass=&NiagaraRibbonRendererProperties}]
 }
 SmokeTrail_01 {
  emitterUpdateScript [
   &EmitterScalability(
    Module.bApplySpawnCountScale=false
    Module.bApplySpawnCountScaleByDistanceFraction=false
    Module.bUseMaxDistance=false
    Module.bUseMinDistance=false
   ) {enabled=false}
   &EmitterState(Life Cycle Mode=::NewEnumerator1, Loop Behavior=::NewEnumerator1, Loop Duration=1.7, UseLoopDelay=false)
   &EmitterLifeCycle() {enabled=false}
   &SpawnRate(SpawnRate=10, Use Spawn Probability=false)
  ]
  particleSpawnScript [
   &InitializeRibbon(
    Color=<rgba 0 0.6667 1 0.4>
    Lifetime=0.7
    Module.Write Color=true
    Module.Write Lifetime=true
    Module.Write Mass=true
    Module.Write Material Random=true
    Module.Write Position=true
    Module.Write RibbonTwist=true
    Module.Write RibbonWidth=true
    Position:Vector3f=&Add_Vector(A=&SimulationPosition(), B=<0 0 0>)
    Ribbon Twist=1
    Ribbon Width=18
   )
   &AccelerationForce(Acceleration=<0 0 50>) {module=AccelerationForce001}
   &AddVelocity(Scale Added Velocity=<3 3 1>, Velocity=&RandomRangeVector(Minimum=<2 2 10>, Maximum=<-2 -2 0>))
   &ApplyInitialForces()
  ]
  particleUpdateScript [
   &ParticleState()
   &AccelerationForce(Acceleration=&RandomRangeVector(Minimum=<0 0 50>, Maximum=<0 0 100>))
   &SolveForcesAndVelocity(Clamp Velocity=false, Limit Acceleration=false)
   &Color(Color=$Particles.Initial.Color)
   /SetVariables(Particles.RibbonWidth=&FloatFromCurve(FloatCurve=DataInterface() {propertyValues {Curve {keys [<at 0 1>,<at 1 0>]}}}, Scale Curve=6)) {enabled=false, isSetParameters=true, module=SetVariables_#1}
   &DynamicMaterialParameters(
    Index 0 Param 1=&FloatFromCurve(FloatCurve=DataInterface() {propertyValues {Curve {keys [<at -0.0046 1.0113>,<at 0.5201 0.2342>,<at 0.7 0.1>]}}})
    Index 0 Param 2=0.1
    Index 0 Param 3=-3
    Index 0 Param 4=&FloatFromCurve(FloatCurve=DataInterface() {propertyValues {Curve {keys [<at -0.0046 1.0113>,<at 1 0>]}}})
    Module.Param0WriteEnabled=true
    Module.Param1WriteEnabled=true
    Module.Param2WriteEnabled=true
    Module.Param3WriteEnabled=true
   )
   &ScaleColor(
    Scale Alpha=1.5
    Scale RGB=&VectorFromFloat(Value=&FloatFromCurve(FloatCurve=DataInterface() {propertyValues {Curve {keys [
          {arriveTangent=-0.0519, interpMode=RCIM_Cubic, leaveTangent=-0.0519, tangentMode=RCTM_User, time=0, value=1}
          {arriveTangent=-0.9379, interpMode=RCIM_Cubic, leaveTangent=-0.9379, tangentMode=RCTM_User, time=0.2578, value=0.2076}
          {arriveTangent=-0.2078, interpMode=RCIM_Cubic, leaveTangent=-0.2078, tangentMode=RCTM_User, time=1, value=0}
         ]}}}))
    ScaleA=true
    ScaleRGB=true
   )
  ]
  rendererClasses [&NiagaraRibbonRendererProperties]
  renderers [{rendererClass=&NiagaraRibbonRendererProperties}]
 }
}
summary {
 systemName=NS_Emote_ManAndMonster_SmokeTrail_01
 emitters [
  {emitterName=SmokeTrail_01, rendererClasses [&NiagaraRibbonRendererProperties]}
  {emitterName=SmokeTrail001, rendererClasses [&NiagaraRibbonRendererProperties]}
 ]
 userVariables [User.IsSprinting {type=NiagaraBool}]
}
